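# Team HUD widget, derived from native AthenaLevelEffects, that shows a "Lvl" label and number over escalating decorative effects.
# Overlaid material images form the backdrop: inherited flame, glow, dark shadow and two rotating spikeball flares.
# On Handle Effects Level Changed, integer thresholds toggle layers between Collapsed and HitTestInvisible: glow above 0, shadow and first flare above 1, second flare above 2.
# The graph never touches the flame or texts, which are inherited widgets.
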WidgetBlueprint /Game/Athena/HUD/Team/AthenaLevelEffects
graphs [EventGraph {nodes [
   n1 = Event Handle Effects Level Changed!() -> (OutputDelegate, then, EffectsLevel:int=0) {is=title K2Node_Event}
   > SetVisibility(
    self=Get Image_Glow() -> (Image_Glow:Image Object Reference)
    InVisibility=Select(Option 0=Collapsed, Option 1=HitTestInvisible, Index="integer > integer"(A=@n1.EffectsLevel, B=0) {is=title K2Node_CallFunction})
   )
   > SetVisibility(
    self=[
     Get Image_Backgroundshadow() -> (Image_Backgroundshadow:Image Object Reference)
     Get Image_spikeball1() -> (Image_spikeball1:Image Object Reference)
    ]
    InVisibility=Select(Option 0=Collapsed, Option 1=HitTestInvisible, Index="integer > integer"(A=@n1.EffectsLevel, B=1))
   )
   > SetVisibility(
    self=Get Image_spikeball2() -> (Image_spikeball2:Image Object Reference)
    InVisibility=Select(Option 0=Collapsed, Option 1=HitTestInvisible, Index="integer > integer"(A=@n1.EffectsLevel, B=2))
   )
  ]}]
variables []
widgets {
 info {inheritedWidgetCount=4, parentClass=/Script/FortniteUI.AthenaLevelEffects, rootWidgetClass=&Overlay, widgetCount=10}
 widgets [
  Overlay_15 {widgetClassPath=&Overlay, widget=Overlay_15}
  Border_Container {
   widgetClassPath=&CommonBorder
   bInherited=true
   parent=Overlay_15
   widget=Border_Container
   properties {verticalAlignment=VAlign_Center, contentColorAndOpacity=<rgba 0.6172 0.2307 1 1>, padding=<ltrb 0 0 0 0>}
   slot {class=OverlaySlot, properties {horizontalAlignment=HAlign_Fill, verticalAlignment=VAlign_Fill}}
  }
  Overlay_12 {
   widgetClassPath=&Overlay
   parent=Border_Container
   widget=Overlay_12
   slot {class=BorderSlot, properties {verticalAlignment=VAlign_Center, padding=<ltrb 0 0 0 0>}}
  }
  Image_Backgroundshadow {
   widgetClassPath=&Image
   bIsVariable=true
   parent=Overlay_12
   widget=Image_Backgroundshadow
   properties {
    renderOpacity=0.5
    visibility=SelfHitTestInvisible
    brush {
     bIsDynamicallyLoaded=false
     drawAs=Image
     imageType=NoImage
     mirroring=NoMirror
     resourceName=None
     resourceObject=/Game/Athena/HUD/LTM/Wax/Materials/M_UI_TextGlow_Inst
     tiling=NoTile
     imageSize=<175.9768 144.8844>
     margin=<ltrb 0 0 0 0>
     outlineSettings {
      bUseBrushTransparency=false
      roundingType=HalfHeightRadius
      width=0
      color {colorUseRule=UseColor_Specified, specifiedColor=<rgba 0 0 0 0>}
      cornerRadii=<0 0 0 0>
     }
     tintColor {colorUseRule=UseColor_Specified, specifiedColor=<rgba 1 1 1 1>}
     uVRegion {bIsValid=false, max=<0 0>, min=<0 0>}
    }
    colorAndOpacity=<rgba 0.0052 0.0052 0.0056 1>
    renderTransform {angle=0, scale=<0.8 0.8>, shear=<0 0>, translation=<0 0>}
   }
   slot {class=OverlaySlot, properties {horizontalAlignment=HAlign_Center, verticalAlignment=VAlign_Center, padding=<ltrb -80 -80 -80 -80>}}
  }
  LazyImage_Flame {
   widgetClassPath=&FortLazyImage
   bInherited=true
   parent=Overlay_12
   widget=LazyImage_Flame
   properties {
    brush {
     bIsDynamicallyLoaded=false
     drawAs=Image
     imageType=NoImage
     mirroring=NoMirror
     resourceName=None
     resourceObject=/Game/Athena/HUD/LTM/Mash/Materials/M_UI_MultiplierFireball_Inst
     tiling=NoTile
     imageSize=<30 50>
     margin=<ltrb 0 0 0 0>
     outlineSettings {
      bUseBrushTransparency=false
      roundingType=HalfHeightRadius
      width=0
      color {colorUseRule=UseColor_Specified, specifiedColor=<rgba 0 0 0 0>}
      cornerRadii=<0 0 0 0>
     }
     tintColor {colorUseRule=UseColor_Specified, specifiedColor=<rgba 1 1 1 1>}
     uVRegion {bIsValid=false, max=<0 0>, min=<0 0>}
    }
    renderTransform {angle=0, scale=<1 1>, shear=<0 0>, translation=<5 8>}
   }
   slot {class=OverlaySlot, properties {horizontalAlignment=HAlign_Fill, verticalAlignment=VAlign_Center, padding=<ltrb -10 -50 -10 -30>}}
  }
  Image_spikeball1 {
   widgetClassPath=&Image
   bIsVariable=true
   parent=Overlay_12
   widget=Image_spikeball1
   properties {
    renderOpacity=0.419
    visibility=SelfHitTestInvisible
    brush {
     bIsDynamicallyLoaded=false
     drawAs=Image
     imageType=NoImage
     mirroring=NoMirror
     resourceName=None
     resourceObject=/Game/Athena/HUD/Materials/Victory/M_Flare_Spikeball_Rotating2
     tiling=NoTile
     imageSize=<80 100>
     margin=<ltrb 0 0 0 0>
     outlineSettings {
      bUseBrushTransparency=false
      roundingType=HalfHeightRadius
      width=0
      color {colorUseRule=UseColor_Specified, specifiedColor=<rgba 0 0 0 0>}
      cornerRadii=<0 0 0 0>
     }
     tintColor {colorUseRule=UseColor_Specified, specifiedColor=<rgba 1 1 1 1>}
     uVRegion {bIsValid=false, max=<0 0>, min=<0 0>}
    }
   }
   slot {class=OverlaySlot, properties {horizontalAlignment=HAlign_Center, verticalAlignment=VAlign_Center, padding=<ltrb -140 -140 -140 -140>}}
  }
  Image_spikeball2 {
   widgetClassPath=&Image
   bIsVariable=true
   parent=Overlay_12
   widget=Image_spikeball2
   properties {
    visibility=SelfHitTestInvisible
    brush {
     bIsDynamicallyLoaded=false
     drawAs=Image
     imageType=NoImage
     mirroring=NoMirror
     resourceName=None
     resourceObject=/Game/Athena/HUD/Materials/Victory/M_Flare_Spikeball_Rotating
     tiling=NoTile
     imageSize=<59.1878 42.1955>
     margin=<ltrb 0 0 0 0>
     outlineSettings {
      bUseBrushTransparency=false
      roundingType=HalfHeightRadius
      width=0
      color {colorUseRule=UseColor_Specified, specifiedColor=<rgba 0 0 0 0>}
      cornerRadii=<0 0 0 0>
     }
     tintColor {colorUseRule=UseColor_Specified, specifiedColor=<rgba 1 1 1 1>}
     uVRegion {bIsValid=false, max=<0 0>, min=<0 0>}
    }
    renderTransform {angle=0, scale=<1.5 1>, shear=<0 0>, translation=<0 0>}
   }
   slot {class=OverlaySlot, properties {horizontalAlignment=HAlign_Center, verticalAlignment=VAlign_Center, padding=<ltrb -40 -40 -40 -40>}}
  }
  Image_Glow {
   widgetClassPath=&Image
   bIsVariable=true
   parent=Overlay_12
   widget=Image_Glow
   properties {
    visibility=SelfHitTestInvisible
    brush {
     bIsDynamicallyLoaded=false
     drawAs=Image
     imageType=NoImage
     mirroring=NoMirror
     resourceName=None
     resourceObject=/Game/Athena/HUD/Materials/Victory/M_WinGlow_White_Inst
     tiling=NoTile
     imageSize=<150 80>
     margin=<ltrb 0 0 0 0>
     outlineSettings {
      bUseBrushTransparency=false
      roundingType=HalfHeightRadius
      width=0
      color {colorUseRule=UseColor_Specified, specifiedColor=<rgba 0 0 0 0>}
      cornerRadii=<0 0 0 0>
     }
     tintColor {colorUseRule=UseColor_Specified, specifiedColor=<rgba 1 1 1 1>}
     uVRegion {bIsValid=false, max=<0 0>, min=<0 0>}
    }
    colorAndOpacity=<rgba 0.2917 0.2672 0 1>
    renderTransform {angle=0, scale=<0.35 0.5>, shear=<0 0>, translation=<0 0>}
   }
   slot {class=OverlaySlot, properties {horizontalAlignment=HAlign_Center, verticalAlignment=VAlign_Center, padding=<ltrb -150 -150 -150 -150>}}
  }
  Text_LvlDesc {
   widgetClassPath=&CommonTextBlock
   bInherited=true
   parent=Overlay_15
   widget=Text_LvlDesc
   properties {
    justification=Center
    text=Lvl
    textTransformPolicy=ToUpper
    font {
     bForceMonospaced=false
     bMaterialIsStencil=false
     fontMaterial=null
     fontObject=&BurbankSmall
     letterSpacing=0
     monospacedWidth=1
     size=8
     skewAmount=0
     typefaceFontName=Black
     outlineSettings {
      bApplyOutlineToDropShadows=true
      bMiteredCorners=false
      bSeparateFillAlpha=true
      outlineMaterial=null
      outlineSize=1
      outlineColor=<rgba 0 0 0 1>
     }
    }
    renderTransform {angle=0, scale=<1 1>, shear=<-12 0>, translation=<-2 -6>}
    shadowColorAndOpacity=<rgba 0 0 0 0.5>
    shadowOffset=<2 2>
   }
   slot {class=OverlaySlot}
  }
  Text_Level {
   widgetClassPath=&CommonTextBlock
   bInherited=true
   parent=Overlay_15
   widget=Text_Level
   properties {
    justification=Center
    text=500
    font {
     bForceMonospaced=false
     bMaterialIsStencil=false
     fontMaterial=null
     fontObject=&Burbank
     letterSpacing=0
     monospacedWidth=1
     size=14
     skewAmount=0
     typefaceFontName=BurbankRegular-Black
     outlineSettings {
      bApplyOutlineToDropShadows=true
      bMiteredCorners=false
      bSeparateFillAlpha=true
      outlineMaterial=null
      outlineSize=1
      outlineColor=<rgba 0 0 0 1>
     }
    }
    renderTransform {angle=0, scale=<1 1>, shear=<-12 0>, translation=<0 2>}
    shadowColorAndOpacity=<rgba 0 0 0 0.5>
    shadowOffset=<2 2>
   }
   slot {class=OverlaySlot, properties {horizontalAlignment=HAlign_Center, verticalAlignment=VAlign_Center}}
  }
 ]
}
