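# Athena HUD squad list widget that stacks the local player's info, a dynamic box of team member entries and an extensions box.
# It listens on the global event router for UIExtension events targeting UnderneathTeammateStatusList, async loads the widget class, then toggles it: removing an existing instance or creating one.
# When a squad member other than the local player disconnects, it plays a sound, with a separate cue if the connected count is one or less.
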WidgetBlueprint /Game/Athena/HUD/Team/AthenaSquadList
graphs [
 AddExtensionWidget {nodes [
   n1 = AddExtensionWidget!() -> (then, ExtensionWidget:User Widget Object Reference) {is=title K2Node_FunctionEntry}
   > Is Valid(InputObject=@n1)
   > n2 = AddChildToVerticalBox(
    self=Get VBox_Extensions() -> (VBox_Extensions:Vertical Box Object Reference)
    Content=Get ExtensionWidget() -> (ExtensionWidget:User Widget Object Reference)
   )
   > SetVerticalAlignment(self:Vertical Box Slot Object Reference=@n2, InVerticalAlignment=VAlign_Fill)
   > SetHorizontalAlignment(self:Vertical Box Slot Object Reference=@n2, InHorizontalAlignment=HAlign_Fill)
   > SetPadding!(
    execute
    self:Vertical Box Slot Object Reference=@n2
    InPadding_Left:float=0
    InPadding_Top:float=0
    InPadding_Right:float=0
    InPadding_Bottom:float=8
   ) -> (then)
  ]}
 AttemptToRemoveExtensionWidget {nodes [
   AttemptToRemoveExtensionWidget!() -> (then, ExtensionWidgetClass:User Widget Class Reference) {is=title K2Node_FunctionEntry}
   > n1 = For Each Loop(Array:Array of Widget Object References=GetAllChildren(self=Get VBox_Extensions())) -> (Array Element:Widget Object Reference)
   Return Node!(execute=@n1.Completed, bExtensionWidgetRemoved:bool=false)
   > Branch(Condition=IsA!(
     Object:Object Reference=@n1
     ObjectClass:Object Class Reference=Get ExtensionWidgetClass() -> (ExtensionWidgetClass:User Widget Class Reference)
    ) -> (ReturnValue:bool))
   > RemoveChildAt!(execute, self:Panel Widget Object Reference=Get VBox_Extensions(), Index:int=@n1.Array Index) -> (then, ReturnValue:bool)
   > Return Node(bExtensionWidgetRemoved=true)
  ]}
 EventGraph {nodes [
   Event Construct()
   n1 = GetGlobalEventRouter()
   > Is Valid(InputObject=@n1)
   > n2 = StartListeningToEvent!(
    execute:Exec
    Target:Gameplay Event Router Component Object Reference=@n1
    EventType:Script Struct Object Reference=/Script/FortniteGame.UIExtension
    Context:Object Reference=GetOwningPlayer()
    EventBubblingRule:EEventBubblingRule Enum
   ) -> (
    then:Exec
    AsyncTaskProxy:Async Action Start Listening To Event Object Reference
    OnEventReceived:Exec
    Payload_Slot:EUIExtensionSlot Enum
    Payload_WidgetClass:User Widget Soft Class Reference
   )
   > Set ListenForUIExtensionAsyncAction(ListenForUIExtensionAsyncAction:Async Action Start Listening To Event Object Reference=@n2) -> (Output_Get:Async Action Start Listening To Event Object Reference)
   Branch(
    execute=@n2.OnEventReceived
    Condition="Equal (Enum)"(A:EUIExtensionSlot Enum=@n2.Payload_Slot, B:EUIExtensionSlot Enum=UnderneathTeammateStatusList)
   )
   > Branch(Condition=IsValidSoftClassReference!(SoftClassReference:"Object Soft Class Reference (by ref)"=@n2.Payload_WidgetClass) -> (ReturnValue:bool))
   > n3 = Async Load Class Asset!(execute:Exec, AssetClass:Object Soft Class Reference=@n2.Payload_WidgetClass) -> (then:Exec, Completed:Exec, Class:Object Class Reference) {is=title K2Node_LoadAssetClass}
   n4 = Cast To UserWidget Class!(execute:Exec=@n3.Completed, Class:Object Class Reference=@n3) -> (then:Exec, CastFailed:Exec, AsUser Widget:User Widget Class Reference) {is=title K2Node_ClassDynamicCast}
   > n5 = AttemptToRemoveExtensionWidget!(execute, self, ExtensionWidgetClass:User Widget Class Reference=@n4) -> (then, bExtensionWidgetRemoved:bool) {is=title K2Node_CallFunction}
   > n6 = Branch(Condition=@n5)
   n7 = Create Widget!(execute:Exec=@n6.else, Class:User Widget Class Reference=@n4, OwningPlayer:Player Controller Object Reference=GetOwningPlayer()) -> (then:Exec, ReturnValue:User Widget Object Reference)
   > AddExtensionWidget!(execute, self, ExtensionWidget:User Widget Object Reference=@n7) -> (then) {is=title K2Node_CallFunction}
   Event Destruct()
   > n8 = Get!(execute:Exec) -> (then:Exec, else:Exec, ListenForUIExtensionAsyncAction:Async Action Start Listening To Event Object Reference)
   > StopListeningToEvent(self:Async Action Start Listening To Event Object Reference=@n8)
   n9 = Event HandleSquadMemberDisconnected!() -> (OutputDelegate, then, Player:Fort Player State Athena Object Reference) {is=title K2Node_Event}
   > n10 = Cast To PlayerController!(execute, Object=GetOwner(self:Actor Object Reference=@n9.Player)) -> (then, CastFailed, AsPlayer Controller:Player Controller Object Reference) {is=title K2Node_DynamicCast}
   > n11 = Branch(Condition=IsLocalController!(self:Controller Object Reference=@n10) -> (ReturnValue:bool))
   n12 = GetSquadConnectedCount!(self:Athena HUDContext Object Reference=Get AthenaHUDContext!() -> (ReturnValue:Athena HUDContext Object Reference) {is=title K2Node_GetBlueprintContextEx}) -> (ReturnValue:int=0)
   PlaySound2D(
    execute=[@n10.CastFailed, @n11.else]
    Sound=Select(
     Option 0:Sound Base Object Reference=Get Sound_SquadMemberDisconnected() -> (Sound_SquadMemberDisconnected:Sound Base Object Reference)
     Option 1:Sound Base Object Reference=Get Sound_LastSquadMemberDisconnected() -> (Sound_LastSquadMemberDisconnected:Sound Base Object Reference)
     Index="integer <= integer"!(A:int=@n12, B:int=1) -> (ReturnValue:bool) {is=title K2Node_CallFunction}
    ) -> (ReturnValue:Sound Base Object Reference)
   )
  ]}
]
variables [
 ListenForUIExtensionAsyncAction {type=AsyncAction_StartListeningToEvent, default=null}
 Sound_LastSquadMemberDisconnected {type=SoundBase, default=/Game/Sounds/Fort_UI_Sounds/Gameplay/Fort_SquadList_LastPlayerLeft_Cue}
 Sound_SquadMemberDisconnected {type=SoundBase, default=/Game/Sounds/Fort_UI_Sounds/Gameplay/Fort_SquadList_PlayerLeft_Cue}
]
widgets {
 info {inheritedWidgetCount=3, parentClass=/Script/FortniteUI.AthenaSquadListBase, rootWidgetClass=&VerticalBox, widgetCount=5}
 widgets [
  VertBox {widgetClassPath=&VerticalBox, bIsVariable=true, widget=VertBox}
  LocalPlayerInfo {
   widgetClassPath=&/AthenaTeamMemberInfo
   bInherited=true
   bIsVariable=true
   parent=VertBox
   widget=LocalPlayerInfo
   properties {bIsLocalPlayer=true, flowDirectionPreference=Inherit, elementTag {tagName=HUD.Player.TeamEntry}}
   slot {class=VerticalBoxSlot, properties {padding=<ltrb 0 0 0 8>}}
  }
  EntryBox_SquadList {
   widgetClassPath=/Script/FortniteUI.FortDynamicEntryBox
   bInherited=true
   bIsVariable=true
   parent=VertBox
   widget=EntryBox_SquadList
   properties {entryBoxType=Vertical, entryWidgetClass=&/AthenaTeamMemberInfo.AthenaTeamMemberInfo_C, entrySpacing=<0 8>}
   slot {class=VerticalBoxSlot, properties {horizontalAlignment=HAlign_Left, padding=<ltrb 0 0 0 8>}}
  }
  VBox_Extensions {
   widgetClassPath=&VerticalBox
   bIsVariable=true
   parent=VertBox
   widget=VBox_Extensions
   slot {class=VerticalBoxSlot, properties {padding=<ltrb 0 0 0 8>}}
  }
  TutorialHighlight {widgetClassPath=&TutorialHighlight}
 ]
}
