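# Small fixed-size widget showing a team member's down-but-not-out (DBNO) state in the Athena team HUD, laid out on the native AthenaTeamMemberDBNOState class.
# A visibility switcher holds a default panel, with a dying-indicator material and an initially collapsed revive-progress timer, and a custom panel.
# The custom panel pairs a lazily loaded widget container with its own revive-progress timer image.
# The Blueprint has no graph logic, only layout and brushes for inherited widgets.
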WidgetBlueprint /Game/Athena/HUD/Team/AthenaTeamMemberDBNOState
graphs [EventGraph {links [], nodes []}]
variables []
widgets {
 info {inheritedWidgetCount=7, parentClass=/Script/FortniteUI.AthenaTeamMemberDBNOState, rootWidgetClass=&SizeBox, widgetCount=8}
 widgets [
  SizeBox_1 {
   widgetClassPath=&SizeBox
   widget=SizeBox_1
   properties {bOverride_HeightOverride=true, bOverride_WidthOverride=true, heightOverride=42, widthOverride=42}
  }
  Switcher_Indicator {
   widgetClassPath=&CommonVisibilitySwitcher
   bInherited=true
   parent=SizeBox_1
   widget=Switcher_Indicator
   slot {class=SizeBoxSlot}
  }
  Panel_DefaultIndicator {
   widgetClassPath=&Overlay
   bInherited=true
   parent=Switcher_Indicator
   widget=Panel_DefaultIndicator
   slot {class=CommonVisibilitySwitcherSlot}
  }
  LazyImage_Indicator {
   widgetClassPath=&FortLazyImage
   bInherited=true
   parent=Panel_DefaultIndicator
   widget=LazyImage_Indicator
   properties {brush {
     bIsDynamicallyLoaded=false
     drawAs=Image
     imageType=NoImage
     mirroring=NoMirror
     resourceName=None
     resourceObject=/Game/Athena/HUD/Art/M_Dying_Indicator
     tiling=NoTile
     imageSize=<42 42>
     margin=<ltrb 0 0 0 0>
     outlineSettings {
      bUseBrushTransparency=false
      roundingType=HalfHeightRadius
      width=0
      color {colorUseRule=UseColor_Specified, specifiedColor=<rgba 0 0 0 0>}
      cornerRadii=<0 0 0 0>
     }
     tintColor {colorUseRule=UseColor_Specified, specifiedColor=<rgba 1 1 1 1>}
     uVRegion {bIsValid=false, max=<0 0>, min=<0 0>}
    }}
   slot {class=OverlaySlot, properties {horizontalAlignment=HAlign_Center, verticalAlignment=VAlign_Center}}
  }
  Image_Timer {
   widgetClassPath=&Image
   bInherited=true
   parent=Panel_DefaultIndicator
   widget=Image_Timer
   properties {
    visibility=Collapsed
    brush {
     bIsDynamicallyLoaded=false
     drawAs=Image
     imageType=NoImage
     mirroring=NoMirror
     resourceName=None
     resourceObject=&/MI_Progress_Revive
     tiling=NoTile
     imageSize=<42 42>
     margin=<ltrb 0 0 0 0>
     outlineSettings {
      bUseBrushTransparency=false
      roundingType=HalfHeightRadius
      width=0
      color {colorUseRule=UseColor_Specified, specifiedColor=<rgba 0 0 0 0>}
      cornerRadii=<0 0 0 0>
     }
     tintColor {colorUseRule=UseColor_Specified, specifiedColor=<rgba 1 1 1 1>}
     uVRegion {bIsValid=false, max=<0 0>, min=<0 0>}
    }
   }
   slot {class=OverlaySlot, properties {horizontalAlignment=HAlign_Center, verticalAlignment=VAlign_Center}}
  }
  Panel_CustomIndicator {
   widgetClassPath=&Overlay
   bInherited=true
   parent=Switcher_Indicator
   widget=Panel_CustomIndicator
   slot {class=CommonVisibilitySwitcherSlot}
  }
  LazyWidget_CustomWidgetContainer {
   widgetClassPath=/Script/CommonUI.CommonLazyWidget
   bInherited=true
   parent=Panel_CustomIndicator
   widget=LazyWidget_CustomWidgetContainer
   slot {class=OverlaySlot, properties {horizontalAlignment=HAlign_Center, verticalAlignment=VAlign_Center}}
  }
  LazyImage_CustomWidgetTimer {
   widgetClassPath=&FortLazyImage
   bInherited=true
   parent=Panel_CustomIndicator
   widget=LazyImage_CustomWidgetTimer
   properties {brush {
     bIsDynamicallyLoaded=false
     drawAs=Image
     imageType=NoImage
     mirroring=NoMirror
     resourceName=None
     resourceObject=&MI_Progress_Revive
     tiling=NoTile
     imageSize=<42 42>
     margin=<ltrb 0 0 0 0>
     outlineSettings {
      bUseBrushTransparency=false
      roundingType=HalfHeightRadius
      width=0
      color {colorUseRule=UseColor_Specified, specifiedColor=<rgba 0 0 0 0>}
      cornerRadii=<0 0 0 0>
     }
     tintColor {colorUseRule=UseColor_Specified, specifiedColor=<rgba 1 1 1 1>}
     uVRegion {bIsValid=false, max=<0 0>, min=<0 0>}
    }}
   slot {class=OverlaySlot, properties {horizontalAlignment=HAlign_Center, verticalAlignment=VAlign_Center}}
  }
 ]
}
