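# Translucent UI font material for player indicator text in the team HUD, with a stroke and a talking-state tint.
# It uses signed-distance font data, smoothstepping the distance into a stroke mask scaled by StrokeOpacity and blending it with the glyph's implicit opacity.
# IsTalking blends the fill from white to light cyan, and multiplying that color by glyph opacity leaves the stroke area dark.
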Material /Game/Athena/HUD/Team/M_Font_PlayerIndicator
graph {
 properties {blendMode=BLEND_Translucent, materialDomain=MD_UI, pixelDepthOffsetMode=PDOM_Legacy, usedShadingModels=Default Lit}
 nodes [
  n1 = FontSignedDistance()
  LinearInterpolate(
   A=Multiply(
    A=SmoothStep(Value=@n1.Smooth Signed Distance) {constMax=-0.04, constMin=-0.105}
    B=ScalarParameter() {defaultValue=0.8, parameterName=StrokeOpacity}
   ) {constB=0.8}
   Alpha=@n1.Implicit Opacity
  )
  Multiply(
   A=LinearInterpolate(
    A=Constant3Vector() {constant=<rgba 1 1 1 1>}
    B=Constant3Vector() {constant=<rgba 0.3698 0.9473 1 1>}
    Alpha=ScalarParameter() {parameterName=IsTalking, sliderMax=1}
   )
   B=@n1.Implicit Opacity
  )
 ]
}
