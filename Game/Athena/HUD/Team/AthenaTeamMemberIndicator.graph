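# Team member indicator HUD widget showing a player name, diamond icon and DBNO (downed) state, built on the native AthenaTeamMemberIndicator class.
# When DBNO becomes true, it plays DBNOIntro, a short scale-and-fade pop on the DBNOState widget.
# On Construct, it resolves the bubble chat class by gameplay tag through the UI data configuration context, creates it asynchronously, then attaches it hidden to a visual attachment anchor.
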WidgetBlueprint /Game/Athena/HUD/Team/AthenaTeamMemberIndicator
animations [DBNOIntro {
  binding_count=1
  length_seconds=0.3
  object=&self:DBNOIntro
  sequence {bindings [DBNOState {
     id=#1
     possessed_class=AthenaTeamMemberDBNOState_C
     tracks [
      {
       type=MovieScene2DTransformTrack
       sections [{
         type=MovieScene2DTransformSection
         end_frame=6
         start_frame=0
         channels [
          Translation.X {type=MovieSceneScriptingFloatChannel, default=40}
          Translation.Y {type=MovieSceneScriptingFloatChannel, default=0}
          Angle {type=MovieSceneScriptingFloatChannel, default=0}
          Scale.X {type=MovieSceneScriptingFloatChannel, keys [[0,2.5],[4,0.75],[6,1]]}
          Scale.Y {type=MovieSceneScriptingFloatChannel, keys [[0,2.5],[4,0.75],[6,1]]}
          Shear.X {type=MovieSceneScriptingFloatChannel, default=0}
          Shear.Y {type=MovieSceneScriptingFloatChannel, default=0}
         ]
        }]
      }
      {
       type=MovieSceneFloatTrack
       display_name=Render Opacity
       sections [{type=MovieSceneFloatSection, end_frame=6, start_frame=0, channels [None {type=MovieSceneScriptingFloatChannel, keys [[0,0],[6,1]]}]}]
      }
     ]
    }]}
 }]
graphs [EventGraph {nodes [
   n1 = Event On Handle DBNO State Changed!() -> (OutputDelegate, then, bIsDBNO:bool) {is=title K2Node_Event}
   > Branch(Condition=@n1.bIsDBNO)
   > PlayAnimation(InAnimation=Get DBNOIntro() -> (DBNOIntro:Widget Animation Object Reference))
   Event Construct()
   > n2 = GetSoftWidgetClassForAlias!(
    execute
    self:Fort UIData Configuration Context Object Reference=Get FortUIDataConfigurationContext!() -> (ReturnValue:Fort UIData Configuration Context Object Reference) {is=title K2Node_GetBlueprintContextEx}
    WidgetAlias:"Gameplay Tag Structure (by ref)"=Get BubbleChatSlotTag() -> (BubbleChatSlotTag:Gameplay Tag Structure)
   ) -> (then, ReturnValue:User Widget Soft Class Reference)
   > Branch(Condition=IsValidSoftClassReference!(SoftClassReference:"Object Soft Class Reference (by ref)"=@n2) -> (ReturnValue:bool))
   > n3 = CreateWidgetAsync!(
    execute:Exec
    UserWidgetSoftClass:User Widget Soft Class Reference=@n2
    OwningPlayer:Player Controller Object Reference
    bSuspendInputUntilComplete:bool=true
   ) -> (then:Exec, OnComplete:Exec, OnFailure:Exec, UserWidget:User Widget Object Reference) {is=title K2Node_AsyncAction}
   n4 = Set BubbleChatWidget(execute=@n3.OnComplete, BubbleChatWidget:User Widget Object Reference=@n3) -> (Output_Get:User Widget Object Reference)
   > AddChild(self=Get VA_BubbleChatAnchor() -> (VA_BubbleChatAnchor:Fort Visual Attachment Object Reference), Content=@n4)
   > SetVisibility(self=Get BubbleChatWidget() -> (BubbleChatWidget:User Widget Object Reference), InVisibility=Hidden)
  ]}]
variables [
 BubbleChatSlotTag {type=GameplayTag, default {tagName=DynamicUI.HUD.Fort.Athena.BubbleChat.Slot}}
 BubbleChatWidget {type=UserWidget, default=null}
]
widgets {
 info {inheritedWidgetCount=7, parentClass=/Script/FortniteUI.AthenaTeamMemberIndicator, rootWidgetClass=&Overlay, widgetCount=9}
 widgets [
  Overlay_2 {widgetClassPath=&Overlay, widget=Overlay_2}
  Image_LeftIcon {
   widgetClassPath=&FortLazyImage
   bInherited=true
   bIsVariable=true
   parent=Overlay_2
   widget=Image_LeftIcon
   properties {
    brush {
     bIsDynamicallyLoaded=false
     drawAs=Image
     imageType=FullColor
     mirroring=NoMirror
     resourceName=None
     resourceObject=/Game/Athena/HUD/Art/T_UI_Indicator_Diamond
     tiling=NoTile
     imageSize=<14 16>
     margin=<ltrb 0 0 0 0>
     outlineSettings {
      bUseBrushTransparency=false
      roundingType=HalfHeightRadius
      width=0
      color {colorUseRule=UseColor_Specified, specifiedColor=<rgba 0 0 0 0>}
      cornerRadii=<0 0 0 0>
     }
     tintColor {colorUseRule=UseColor_Specified, specifiedColor=<rgba 1 1 1 1>}
     uVRegion {bIsValid=false, max=<0 0>, min=<0 0>}
    }
    renderTransform {angle=0, scale=<1 1>, shear=<0 0>, translation=<-16 0>}
   }
   slot {class=OverlaySlot, properties {verticalAlignment=VAlign_Center}}
  }
  Text_PlayerName {
   widgetClassPath=&TextBlock
   bInherited=true
   parent=Overlay_2
   widget=Text_PlayerName
   properties {
    text="{MaxCharacterLim}"
    visibility=SelfHitTestInvisible
    font {
     bForceMonospaced=false
     bMaterialIsStencil=true
     fontMaterial=&/M_Font_PlayerIndicator
     fontObject=&Burbank
     letterSpacing=0
     monospacedWidth=1
     size=14
     skewAmount=0
     typefaceFontName=BurbankRegular-Black
     outlineSettings {
      bApplyOutlineToDropShadows=false
      bMiteredCorners=false
      bSeparateFillAlpha=false
      outlineMaterial=null
      outlineSize=0
      outlineColor=<rgba 0 0 0 1>
     }
    }
    renderTransform {angle=0, scale=<1 1>, shear=<-8 0>, translation=<0 0>}
    shadowOffset=<0 0>
   }
   slot {class=OverlaySlot, properties {horizontalAlignment=HAlign_Center, verticalAlignment=VAlign_Center}}
  }
  DBNOState {
   widgetClassPath=&/AthenaTeamMemberDBNOState
   bInherited=true
   parent=Overlay_2
   widget=DBNOState
   properties {bIsVolatile=true, visibility=Collapsed, renderTransform {angle=0, scale=<1 1>, shear=<0 0>, translation=<40 0>}}
   slot {class=OverlaySlot, properties {horizontalAlignment=HAlign_Right, verticalAlignment=VAlign_Center, padding=<ltrb 0 -40 0 -40>}}
  }
  VA_BubbleChatAnchor {
   widgetClassPath=&FortVisualAttachment
   bIsVariable=true
   parent=Overlay_2
   widget=VA_BubbleChatAnchor
   properties {renderTransform {angle=0, scale=<1 1>, shear=<0 0>, translation=<0 -48>}}
   slot {class=OverlaySlot, properties {horizontalAlignment=HAlign_Center, verticalAlignment=VAlign_Bottom}}
  }
  Panel_Extensions {
   widgetClassPath=&Overlay
   bInherited=true
   parent=Overlay_2
   widget=Panel_Extensions
   properties {renderTransform {angle=0, scale=<1 1>, shear=<0 0>, translation=<0 -70>}}
   slot {class=OverlaySlot, properties {horizontalAlignment=HAlign_Center, verticalAlignment=VAlign_Center}}
  }
  DBNOCountdown {widgetClassPath=/Script/FortniteUI.AthenaTeamMemberDBNOCountdown, bInherited=true, widget=null}
  AnimationsPanel_DBNO {widgetClassPath=/Script/FortniteUI.FortSimpleWidgetAnimationsPanel, bInherited=true, widget=null}
  TutorialHighlight {widgetClassPath=&TutorialHighlight}
 ]
}
