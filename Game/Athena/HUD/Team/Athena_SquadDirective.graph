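# Squad HUD banner that prompts players to collect a teammate's Reboot Card or use a Reboot Van to restore squadmates.
# When owner-enabled, ForceUpdateSquadListDisplay caches the player state, builds uppercase rich text, and broadcasts the card's remaining time.
# EventIntro selects the message by update type, plays the slide-and-fade Intro, and restarts a timer.
# The timer fires Outro, which plays the exit animation and clears the enable flag.
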WidgetBlueprint /Game/Athena/HUD/Team/Athena_SquadDirective
animations [
 Intro {
  binding_count=3
  length_seconds=0.3
  object=&self:Intro
  sequence {bindings [
    Image_Background {
     id=#1
     possessed_class=Image
     tracks [
      {
       type=MovieScene2DTransformTrack
       sections [{
         type=MovieScene2DTransformSection
         end_frame=6
         start_frame=0
         channels [Translation.X {type=MovieSceneScriptingFloatChannel, keys [[0,-500],[6,0]]}]
        }]
      }
      {
       type=MovieSceneFloatTrack
       display_name=Render Opacity
       sections [{type=MovieSceneFloatSection, end_frame=3, start_frame=0, channels [None {type=MovieSceneScriptingFloatChannel, keys [[0,0],[3,1]]}]}]
      }
     ]
    }
    Overlay_TextContainer {
     id=#2
     possessed_class=Overlay
     tracks [
      {
       type=MovieSceneFloatTrack
       display_name=Render Opacity
       sections [{type=MovieSceneFloatSection, end_frame=5, start_frame=0, channels [None {type=MovieSceneScriptingFloatChannel, keys [[2,0],[5,1]]}]}]
      }
      {
       type=MovieScene2DTransformTrack
       sections [{
         type=MovieScene2DTransformSection
         end_frame=6
         start_frame=0
         channels [Translation.X {type=MovieSceneScriptingFloatChannel, keys [[0,-500],[6,0]]}]
        }]
      }
     ]
    }
    Overlay_RebootIcon {
     id=#3
     possessed_class=Overlay
     tracks [
      {
       type=MovieSceneFloatTrack
       display_name=Render Opacity
       sections [{type=MovieSceneFloatSection, end_frame=3, start_frame=0, channels [None {type=MovieSceneScriptingFloatChannel, keys [[0,0],[3,1]]}]}]
      }
      {
       type=MovieScene2DTransformTrack
       sections [{
         type=MovieScene2DTransformSection
         end_frame=3
         start_frame=0
         channels [
          Scale.X {type=MovieSceneScriptingFloatChannel, keys [[0,1.2],[3,1]]}
          Scale.Y {type=MovieSceneScriptingFloatChannel, keys [[0,1.2],[3,1]]}
         ]
        }]
      }
     ]
    }
   ]}
 }
 OutroAnim {
  binding_count=4
  length_seconds=0.3517
  object=&self:OutroAnim
  sequence {bindings [
    Overlay_RebootIcon {
     id=#4
     possessed_class=Overlay
     tracks [
      {
       type=MovieSceneFloatTrack
       display_name=Render Opacity
       sections [{type=MovieSceneFloatSection, end_frame=3, start_frame=0, channels [None {type=MovieSceneScriptingFloatChannel, keys [[0,1],[3,0]]}]}]
      }
      {
       type=MovieScene2DTransformTrack
       sections [{
         type=MovieScene2DTransformSection
         end_frame=4
         start_frame=0
         channels [
          Scale.X {type=MovieSceneScriptingFloatChannel, keys [[0,1],[3,1.2]]}
          Scale.Y {type=MovieSceneScriptingFloatChannel, keys [[0,1],[3,1.2]]}
         ]
        }]
      }
     ]
    }
    Root {
     id=#5
     possessed_class=Overlay
     tracks [{
       type=MovieSceneByteTrack
       display_name=Visibility
       sections [{type=MovieSceneByteSection, end_frame=7, start_frame=0, channels [None {type=MovieSceneScriptingByteChannel, keys [[0,3],[7,2],[7.0337,1]]}]}]
      }]
    }
    Image_Background {
     id=#6
     possessed_class=Image
     tracks [
      {
       type=MovieScene2DTransformTrack
       sections [{
         type=MovieScene2DTransformSection
         end_frame=6
         start_frame=0
         channels [Translation.X {type=MovieSceneScriptingFloatChannel, keys [[0,0],[6,-500]]}]
        }]
      }
      {
       type=MovieSceneFloatTrack
       display_name=Render Opacity
       sections [{type=MovieSceneFloatSection, end_frame=3, start_frame=0, channels [None {type=MovieSceneScriptingFloatChannel, keys [[0,1],[3,0]]}]}]
      }
     ]
    }
    Overlay_TextContainer {
     id=#7
     possessed_class=Overlay
     tracks [
      {
       type=MovieSceneFloatTrack
       display_name=Render Opacity
       sections [{type=MovieSceneFloatSection, end_frame=3, start_frame=0, channels [None {type=MovieSceneScriptingFloatChannel, keys [[0,1],[3,0]]}]}]
      }
      {
       type=MovieScene2DTransformTrack
       sections [{
         type=MovieScene2DTransformSection
         end_frame=6
         start_frame=0
         channels [Translation.X {type=MovieSceneScriptingFloatChannel, keys [[0,0],[6,-500]]}]
        }]
      }
     ]
    }
   ]}
 }
]
graphs [
 Cleanup Directive State Anim Event {nodes [Cleanup Directive State Anim Event() {is=title K2Node_FunctionEntry}]}
 EnableSquadDirective {nodes [
   n1 = EnableSquadDirective!() -> (then, EnableSquadDirective:bool=~) {is=title K2Node_FunctionEntry}
   > Set EnableSquadDirectiveFromOwner(EnableSquadDirectiveFromOwner:bool=@n1) -> (Output_Get:bool)
  ]}
 EventGraph {nodes [
   n1 = EventIntro!() -> (OutputDelegate, then, UpdateType:EAthenaSquadListUpdateType Enum=None) {is=title K2Node_CustomEvent}
   > n2 = Branch(Condition="Not Equal (Enum)"!(A:EAthenaSquadListUpdateType Enum=@n1.UpdateType, B:EAthenaSquadListUpdateType Enum=None) -> (ReturnValue:bool))
   > SetVisibility(InVisibility=HitTestInvisible)
   > Clear and Invalidate Timer by Handle(Handle=Get OutroTimerHandle() -> (OutroTimerHandle:Timer Handle Structure))
   > SetText(
    self=Get RichText_DisplayedMessage() -> (RichText_DisplayedMessage:Common Rich Text Block Object Reference)
    InText=Select!(
     None:Text="LOCTEXT(\"None\")"
     CanResurrect:Text=Get CanResurrectText() -> (CanResurrectText:Text)
     FindResurrectChip:Text=Get FindResurrectChipText() -> (FindResurrectChipText:Text)
     Index:EAthenaSquadListUpdateType Enum=@n1.UpdateType
    ) -> (ReturnValue:Text)
   )
   > PlayAnimation(InAnimation=Get Intro())
   > n3 = Set Visibility(Visibility=HitTestInvisible, self=Get Root() -> (Root:Overlay Object Reference))
   n4 = Outro() {is=title K2Node_CustomEvent}
   n5 = GetRuntimeOptions()
   n6 = Set Timer by Event(
    execute=@n3.then
    Delegate=@n4
    Time=Select(
     None:double=GetRebootDirectiveDisplayTime!(self:Fort Runtime Options Object Reference=@n5) -> (ReturnValue:float)
     CanResurrect:double=5
     FindResurrectChip:double=5
     Index=@n1.UpdateType
    ) -> (ReturnValue:double)
    bLooping=false
   )
   > Set OutroTimerHandle(OutroTimerHandle:Timer Handle Structure=@n6) -> (Output_Get:Timer Handle Structure)
   SetVisibility(execute=@n2.else, InVisibility=Collapsed)
   TestChipUI() {is=title K2Node_CustomEvent}
   > n7 = Set test(test:EAthenaSquadListUpdateType Enum=FindResurrectChip) -> (Output_Get:EAthenaSquadListUpdateType Enum=None)
   > EventIntro!(execute, self, UpdateType:EAthenaSquadListUpdateType Enum=@n7) -> (then) {is=title K2Node_CallFunction}
   n8 = ForceUpdateSquadListDisplay!() -> (OutputDelegate, then, Update Type:EAthenaSquadListUpdateType Enum=None, PS:Fort Player State Athena Object Reference) {is=title K2Node_CustomEvent}
   > Branch(Condition=Get EnableSquadDirectiveFromOwner() -> (EnableSquadDirectiveFromOwner:bool))
   > Is Valid(InputObject=@n8.PS)
   > SetVisibility(InVisibility=Collapsed)
   > Branch(Condition="Not Equal (Enum)"(A=@n8.Update Type, B=None))
   > Set CachedPlayerState(CachedPlayerState:Fort Player State Athena Object Reference=@n8.PS) -> (Output_Get:Fort Player State Athena Object Reference)
   n9 = "To Text (String)"(InString=GetPlayerName!(self:Player State Object Reference=@n8.PS) -> (ReturnValue:String))
   > Set FindResurrectChipText(FindResurrectChipText:Text=TextToUpper!(InText:"Text (by ref)"=Format Text!(Format="LOCTEXT(\"Collect <Name>{x}\\'s</> Reboot Card\")", x:Text=@n9) -> (Result)) -> (ReturnValue:Text)) -> (Output_Get:Text)
   > Set CanResurrectText(CanResurrectText:Text=TextToUpper(InText=Format Text!(Format="LOCTEXT(\"Use a <Name>Reboot Van</> to Restore Squadmates!\")") -> (Result))) -> (Output_Get:Text)
   n10 = Get CachedPlayerState() -> (CachedPlayerState:Fort Player State Athena Object Reference)
   > Call Received Despawn Time!(
    execute:Exec
    self:Athena Squad Directive Object Reference
    RemainingTime:int=Truncate(A=GetRemainingResurrectionChipExpirationTime!(self:Fort Player State Athena Object Reference=@n10) -> (ReturnValue:float))
   ) -> (then:Exec) {is=title K2Node_CallDelegate}
   > EventIntro(UpdateType=@n8.Update Type)
   GetValueAtLevel!(Input:"Scalable Float Structure (by ref)"=Get Row_ExpirationLength() -> (Row_ExpirationLength:Scalable Float Structure), Level:float=0) -> (ReturnValue:float)
   Cleanup Directive State Anim Event() {is=title K2Node_CallFunction}
   Set EnableSquadDirectiveFromOwner(execute=@n11.then, EnableSquadDirectiveFromOwner=false)
   n11 = PlayAnimation(execute=@n4.then, InAnimation=Get OutroAnim() -> (OutroAnim:Widget Animation Object Reference))
  ]}
 ReceivedDespawnTime {nodes [ReceivedDespawnTime!() -> (then, RemainingTime:int) {is=title K2Node_FunctionEntry}]}
]
variables [
 CachedPlayerState {type=FortPlayerStateAthena, default=null}
 CanResurrectText {type=string, default="DELIVER THE CARD TO A REBOOT VAN!"}
 EnableSquadDirectiveFromOwner {type=boolean}
 FindResurrectChipText {type=string, default="A REBOOT CARD WAS DROPPED!"}
 OutroTimerHandle {type=TimerHandle}
 ReceivedDespawnTime {}
 Row_ExpirationLength {
  type=ScalableFloat
  default {registryType=None, value=1, curve {curveTable=/Game/Athena/Balance/DataTables/AthenaGameData, rowName=Default.SCMachine.PickupLife}}
 }
 test {type=EAthenaSquadListUpdateType, default=None}
]
widgets {
 info {inheritedWidgetCount=1, parentClass=/Script/FortniteUI.AthenaRebootChipDirective, rootWidgetClass=&Overlay, widgetCount=15}
 widgets [
  Root {widgetClassPath=&Overlay, bIsVariable=true, widget=Root}
  Image_Background {
   widgetClassPath=&Image
   bIsVariable=true
   parent=Root
   widget=Image_Background
   properties {
    brush {
     bIsDynamicallyLoaded=false
     drawAs=Image
     imageType=NoImage
     mirroring=NoMirror
     resourceName=None
     resourceObject=/Game/Athena/HUD/Team/Materials/MI_UI_SquadDirectiveBackground
     tiling=NoTile
     imageSize=<32 32>
     margin=<ltrb 0 0 0 0>
     outlineSettings {
      bUseBrushTransparency=false
      roundingType=FixedRadius
      width=0
      color {colorUseRule=UseColor_Specified, specifiedColor=<rgba 0 0 0 0>}
      cornerRadii=<0 10 10 0>
     }
     tintColor {colorUseRule=UseColor_Specified, specifiedColor=<rgba 1 1 1 1>}
     uVRegion {bIsValid=false, max=<0 0>, min=<0 0>}
    }
    colorAndOpacity=<rgba 1 1 1 0.95>
   }
   slot {class=OverlaySlot, properties {horizontalAlignment=HAlign_Fill, verticalAlignment=VAlign_Fill}}
  }
  CommonBorder_0 {
   widgetClassPath=&CommonBorder
   parent=Root
   widget=CommonBorder_0
   properties {padding=<ltrb 50 5 30 5>}
   slot {class=OverlaySlot, properties {padding=<ltrb 0 0 10 0>}}
  }
  HorizontalBox_0 {
   widgetClassPath=&HorizontalBox
   parent=CommonBorder_0
   widget=HorizontalBox_0
   slot {class=BorderSlot, properties {padding=<ltrb 50 5 30 5>}}
  }
  CommonWidgetSwitcher_IconType {
   widgetClassPath=&CommonWidgetSwitcherLegacy
   bIsVariable=true
   parent=HorizontalBox_0
   widget=CommonWidgetSwitcher_IconType
   slot {class=HorizontalBoxSlot, properties {padding=<ltrb 5 0 0 0>}}
  }
  HorizontalBox_1 {
   widgetClassPath=&HorizontalBox
   parent=CommonWidgetSwitcher_IconType
   widget=HorizontalBox_1
   slot {class=WidgetSwitcherSlot, properties {horizontalAlignment=HAlign_Center, verticalAlignment=VAlign_Center}}
  }
  Overlay_RebootIcon {widgetClassPath=&Overlay, parent=HorizontalBox_1, widget=Overlay_RebootIcon, slot {class=HorizontalBoxSlot}}
  Image_PlayerIconOverlay {
   widgetClassPath=&Image
   bIsVariable=true
   parent=Overlay_RebootIcon
   widget=Image_PlayerIconOverlay
   properties {
    brush {
     bIsDynamicallyLoaded=false
     drawAs=Image
     imageType=NoImage
     mirroring=NoMirror
     resourceName=None
     resourceObject=/FortUILibrary/Foundation/Art_ToolKit/Tools/M_UI_Circle_White
     tiling=NoTile
     imageSize=<53 53>
     margin=<ltrb 0 0 0 0>
     outlineSettings {
      bUseBrushTransparency=false
      roundingType=HalfHeightRadius
      width=0
      color {colorUseRule=UseColor_Specified, specifiedColor=<rgba 0 0 0 0>}
      cornerRadii=<0 0 0 0>
     }
     tintColor {colorUseRule=UseColor_Specified, specifiedColor=<rgba 1 1 1 1>}
     uVRegion {bIsValid=false, max=<0 0>, min=<0 0>}
    }
    colorAndOpacity=<rgba 0 0 0 0.4>
   }
   slot {class=OverlaySlot}
  }
  Image_Reboot {
   widgetClassPath=&Image
   bIsVariable=true
   parent=Overlay_RebootIcon
   widget=Image_Reboot
   properties {brush {
     bIsDynamicallyLoaded=false
     drawAs=Image
     imageType=NoImage
     mirroring=NoMirror
     resourceName=None
     resourceObject=/Game/Athena/HUD/Materials/M_UI_RebootIcon_Vis_Inst
     tiling=NoTile
     imageSize=<22 22>
     margin=<ltrb 0 0 0 0>
     outlineSettings {
      bUseBrushTransparency=false
      roundingType=HalfHeightRadius
      width=0
      color {colorUseRule=UseColor_Specified, specifiedColor=<rgba 0 0 0 0>}
      cornerRadii=<0 0 0 0>
     }
     tintColor {colorUseRule=UseColor_Specified, specifiedColor=<rgba 1 1 1 1>}
     uVRegion {bIsValid=false, max=<0 0>, min=<0 0>}
    }}
   slot {class=OverlaySlot, properties {horizontalAlignment=HAlign_Fill, verticalAlignment=VAlign_Fill, padding=<ltrb 0 0 0 1>}}
  }
  Image_ChipCarried {
   widgetClassPath=&Image
   bIsVariable=true
   parent=CommonWidgetSwitcher_IconType
   widget=Image_ChipCarried
   properties {brush {
     bIsDynamicallyLoaded=false
     drawAs=Image
     imageType=NoImage
     mirroring=NoMirror
     resourceName=None
     resourceObject=/Game/Athena/HUD/Materials/M_UI_RebootChip_Icon_Carried
     tiling=NoTile
     imageSize=<32 33>
     margin=<ltrb 0 0 0 0>
     outlineSettings {
      bUseBrushTransparency=false
      roundingType=HalfHeightRadius
      width=0
      color {colorUseRule=UseColor_Specified, specifiedColor=<rgba 0 0 0 0>}
      cornerRadii=<0 0 0 0>
     }
     tintColor {colorUseRule=UseColor_Specified, specifiedColor=<rgba 1 1 1 1>}
     uVRegion {bIsValid=false, max=<0 0>, min=<0 0>}
    }}
   slot {class=WidgetSwitcherSlot, properties {horizontalAlignment=HAlign_Center, verticalAlignment=VAlign_Center}}
  }
  Image_ChipDropped {
   widgetClassPath=&Image
   bIsVariable=true
   parent=CommonWidgetSwitcher_IconType
   widget=Image_ChipDropped
   properties {brush {
     bIsDynamicallyLoaded=false
     drawAs=Image
     imageType=NoImage
     mirroring=NoMirror
     resourceName=None
     resourceObject=/Game/Athena/HUD/Materials/M_UI_RebootChip_Icon_Dropped
     tiling=NoTile
     imageSize=<32 33>
     margin=<ltrb 0 0 0 0>
     outlineSettings {
      bUseBrushTransparency=false
      roundingType=HalfHeightRadius
      width=0
      color {colorUseRule=UseColor_Specified, specifiedColor=<rgba 0 0 0 0>}
      cornerRadii=<0 0 0 0>
     }
     tintColor {colorUseRule=UseColor_Specified, specifiedColor=<rgba 1 1 1 1>}
     uVRegion {bIsValid=false, max=<0 0>, min=<0 0>}
    }}
   slot {class=WidgetSwitcherSlot, properties {horizontalAlignment=HAlign_Center, verticalAlignment=VAlign_Center}}
  }
  Overlay_TextContainer {widgetClassPath=&Overlay, parent=HorizontalBox_0, widget=Overlay_TextContainer, slot {class=HorizontalBoxSlot}}
  HorizontalBox_2 {
   widgetClassPath=&HorizontalBox
   parent=Overlay_TextContainer
   widget=HorizontalBox_2
   slot {class=OverlaySlot, properties {horizontalAlignment=HAlign_Center, verticalAlignment=VAlign_Center, padding=<ltrb 10 0 0 0>}}
  }
  RichText_DisplayedMessage {
   widgetClassPath=&CommonRichTextBlock
   bIsVariable=true
   parent=HorizontalBox_2
   widget=RichText_DisplayedMessage
   properties {
    inlineIconDisplayMode=TextOnly
    mobileTextBlockScale=1.2
    text="COLLECT <Name>PLAYER 2024'S</> REBOOT CARD"
    textStyleSet=&/RebootTextStyles
    defaultTextStyleOverride {
     overflowPolicy=Clip
     transformPolicy=None
     colorAndOpacity {colorUseRule=UseColor_Specified, specifiedColor=<rgba 1 1 1 1>}
     font {
      bForceMonospaced=false
      bMaterialIsStencil=false
      fontMaterial=null
      fontObject=&Burbank
      letterSpacing=0
      monospacedWidth=1
      size=28
      skewAmount=0
      typefaceFontName=Burbank-Black
      outlineSettings {
       bApplyOutlineToDropShadows=false
       bMiteredCorners=false
       bSeparateFillAlpha=false
       outlineMaterial=/Game/Athena/HUD/Art/Materials/M_HUD_Sheen_Inst
       outlineSize=1
       outlineColor=<rgba 1 1 1 1>
      }
     }
     highlightColor {colorUseRule=UseColor_Specified, specifiedColor=<rgba 0 0 0 1>}
     highlightShape {
      bIsDynamicallyLoaded=false
      drawAs=Image
      imageType=NoImage
      mirroring=NoMirror
      resourceName=None
      resourceObject=null
      tiling=NoTile
      imageSize=<32 32>
      margin=<ltrb 0 0 0 0>
      outlineSettings {
       bUseBrushTransparency=false
       roundingType=HalfHeightRadius
       width=0
       color {colorUseRule=UseColor_Specified, specifiedColor=<rgba 0 0 0 0>}
       cornerRadii=<0 0 0 0>
      }
      tintColor {colorUseRule=UseColor_Specified, specifiedColor=<rgba 1 1 1 1>}
      uVRegion {bIsValid=false, max=<0 0>, min=<0 0>}
     }
     selectedBackgroundColor {colorUseRule=UseColor_Foreground, specifiedColor=<rgba 1 0 1 1>}
     shadowColorAndOpacity=<rgba 0 0.0307 0.0288 0.8>
     shadowOffset=<3 3>
     strikeBrush {
      bIsDynamicallyLoaded=false
      drawAs=Image
      imageType=NoImage
      mirroring=NoMirror
      resourceName=None
      resourceObject=null
      tiling=NoTile
      imageSize=<32 32>
      margin=<ltrb 0 0 0 0>
      outlineSettings {
       bUseBrushTransparency=false
       roundingType=HalfHeightRadius
       width=0
       color {colorUseRule=UseColor_Specified, specifiedColor=<rgba 0 0 0 0>}
       cornerRadii=<0 0 0 0>
      }
      tintColor {colorUseRule=UseColor_Specified, specifiedColor=<rgba 1 1 1 1>}
      uVRegion {bIsValid=false, max=<0 0>, min=<0 0>}
     }
     underlineBrush {
      bIsDynamicallyLoaded=false
      drawAs=Image
      imageType=NoImage
      mirroring=NoMirror
      resourceName=None
      resourceObject=null
      tiling=NoTile
      imageSize=<32 32>
      margin=<ltrb 0 0 0 0>
      outlineSettings {
       bUseBrushTransparency=false
       roundingType=HalfHeightRadius
       width=0
       color {colorUseRule=UseColor_Specified, specifiedColor=<rgba 0 0 0 0>}
       cornerRadii=<0 0 0 0>
      }
      tintColor {colorUseRule=UseColor_Specified, specifiedColor=<rgba 1 1 1 1>}
      uVRegion {bIsValid=false, max=<0 0>, min=<0 0>}
     }
    }
   }
   slot {class=HorizontalBoxSlot}
  }
  TutorialHighlight {widgetClassPath=&TutorialHighlight}
 ]
}
