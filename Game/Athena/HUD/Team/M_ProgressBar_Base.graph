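# Base translucent UI material for a progress bar in the Athena team HUD.
# A step of a horizontal UV gradient against Progress separates the ColorA-to-ColorB gradient fill, optionally overridden to solid ColorB, from the Unfilled Color, and also drives opacity.
# An optional UseSheenPulse switch blends in SheenColor through a shine texture slid by Position, with slightly rotated 9-slice UVs computed via customized UVs.
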Material /Game/Athena/HUD/Team/M_ProgressBar_Base
graph {
 properties {
  blendMode=BLEND_Translucent
  materialDomain=MD_UI
  numCustomizedUVs=2
  pixelDepthOffsetMode=PDOM_Legacy
  shadingModel=MSM_Unlit
  usedShadingModels=Unlit
 }
 nodes [
  MaterialFunctionCall(
   UVs=MaterialFunctionCall() {materialFunction=&GetUserInterfaceUV}.9-Slice UV
   "Rotation Angle (0-1)"=Constant() {r=0.02}
  ) {materialFunction=&CustomRotator}
  n1 = Step(Y=NamedRerouteUsage(), X=ScalarParameter() {defaultValue=0.75, parameterName=Progress})
  n2 = VectorParameter() {parameterName=Unfilled Color, defaultValue=<rgba 0 0 0 0.25>}
  Max(A=@n2.A, B=@n1)
  NamedRerouteDeclaration(None=MaterialFunctionCall(UV Channel=TextureCoordinate()) {materialFunction=&LinearGradient}.UGradient) {name=Gradient, nodeColor=<rgba 0 0.4844 1 1>}
  n3 = VectorParameter() {parameterName=ColorA, defaultValue=<rgba 1 0 0 1>}
  n4 = VectorParameter() {parameterName=ColorB, defaultValue=<rgba 0 1 0 1>}
  n5 = ScalarParameter() {parameterName=OverrideWithColorB}
  n6 = LinearInterpolate(A=@n2.RGB, B=LinearInterpolate(A=LinearInterpolate(A=@n3.RGB, B=@n4.RGB, Alpha=@n5), B=@n4.RGB, Alpha=NamedRerouteUsage()), Alpha=@n1)
  n7 = LinearInterpolate(Alpha=ScalarParameter() {parameterName=Position}) {constA=1, constAlpha=0.2, constB=-1}
  n8 = Add(A=TextureCoordinate() {coordinateIndex=1}, B=AppendVector(A=@n7, B=@n7))
  StaticSwitchParameter(
   True=LinearInterpolate(
    A=@n6
    B=VectorParameter() {parameterName=SheenColor, defaultValue=<rgba 1 1 1 1>}.RGB
    Alpha=TextureSample(UVs=@n8) {samplerSource=SSM_Clamp_WorldGroupSettings, texture=/Game/UI/Foundation/Textures/Pattern/T-UI-Shine}.R
   )
   False=@n6
  ) {parameterName=UseSheenPulse}
 ]
}
