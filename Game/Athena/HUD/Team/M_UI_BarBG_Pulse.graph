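# Translucent UI material for a team HUD bar background that pulses between a base colour and a pulse colour.
# Time wraps over a period taken from SpeedRange, then a LinearPower exponent, clamp and inversion shape it into a repeating falloff.
# Severity scales that pulse as the colour blend, while the pulse also drives an opacity from OpacityRange, floored at the base colour's alpha.
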Material /Game/Athena/HUD/Team/M_UI_BarBG_Pulse
graph {
 properties {blendMode=BLEND_Translucent, materialDomain=MD_UI, pixelDepthOffsetMode=PDOM_Legacy, usedShadingModels=Default Lit}
 nodes [
  Constant() {r=1}
  n1 = VectorParameter() {parameterName=SpeedRange, defaultValue=<rgba 0 1 0 1>}
  n2 = LinearInterpolate(A=@n1.R, B=@n1.R, Alpha=ScalarParameter() {parameterName=Severity})
  n3 = Fmod(A=Time(), B=@n2)
  Divide(A=@n3, B=@n2)
  n4 = ScalarParameter() {defaultValue=1, parameterName=LinearPower}
  n5 = OneMinus(None=Clamp(None=Power(Base=@n3, Exp=@n4)))
  n6 = ScalarParameter() {parameterName=Severity, sliderMax=1}
  n7 = VectorParameter() {parameterName=BarBaseColor, defaultValue=<rgba 0.0052 0.0052 0.0056 0.4>}
  LinearInterpolate(A=@n7.RGB, B=VectorParameter() {parameterName=PulseColor, defaultValue=<rgba 0.1562 0 0 1>}.RGB, Alpha=Clamp(None=Multiply(A=@n5, B=@n6)))
  n8 = VectorParameter() {parameterName=OpacityRange, defaultValue=<rgba 0 1 0 1>}
  n9 = LinearInterpolate(B=LinearInterpolate(A=@n8.R, B=@n8.R, Alpha=ScalarParameter() {parameterName=Severity}), Alpha=@n5)
  Reroute(None=Clamp(None=Max(A=@n7.A, B=@n9)))
 ]
}
