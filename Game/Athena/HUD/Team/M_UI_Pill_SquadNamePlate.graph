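# Translucent UI material for the squad nameplate pill in the Athena team HUD, built as a procedural white radial mask with no textures.
# UVs are shifted and scaled to a -1 to 1 range, then fed into a SphereMask centered at the origin.
# Radius and Hardness parameters control the mask's size and edge softness, and the result multiplies a constant white color.
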Material /Game/Athena/HUD/Team/M_UI_Pill_SquadNamePlate
graph {
 properties {blendMode=BLEND_Translucent, materialDomain=MD_UI, pixelDepthOffsetMode=PDOM_Legacy, usedShadingModels=Default Lit}
 nodes [
  n1 = Subtract(A=TextureCoordinate()) {constB=0.5}
  Multiply(
   A=Constant3Vector() {constant=<rgba 1 1 1 0>}
   B=SphereMask(
    A=Multiply(A=@n1) {constB=2}
    B=Constant2Vector()
    Radius=ScalarParameter() {defaultValue=0.5, parameterName=Radius}
    Hardness=ScalarParameter() {defaultValue=0.945, parameterName=Hardness}
   )
  )
 ]
}
