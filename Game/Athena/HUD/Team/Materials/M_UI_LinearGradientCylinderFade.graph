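# Translucent UI material in the team HUD folder that builds a three-color linear gradient and a soft cylinder-style fade mask.
# A linear gradient function with adjustable stops drives a ColorA/B/C lerp, while a reflected gradient, eased with EaseInSine and scaled by Opacity, forms the fade.
# Both gradients share offset, rotation and center parameter names, but the fade uses Center X for both axes.
# GradientWidth, GradientDensity and GradientPower feed nothing.
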Material /Game/Athena/HUD/Team/Materials/M_UI_LinearGradientCylinderFade
graph {
 properties {blendMode=BLEND_Translucent, materialDomain=MD_UI, pixelDepthOffsetMode=PDOM_Legacy, usedShadingModels=Default Lit}
 nodes [
  n1 = VectorParameter() {
   group=Gradient Settings
   parameterName=GradientRotationCenter
   sortPriority=5
   channelNames {a="", b="", g=Center Y, r=Center X}
   defaultValue=<rgba 0.5 0.5 0 0>
  }
  MaterialFunctionCall(
   A=VectorParameter() {group=Color Settings, parameterName=ColorA, defaultValue=<rgba 1 0 0 1>}.RGB
   B=VectorParameter() {group=Color Settings, parameterName=ColorB, defaultValue=<rgba 0 1 0 1>}.RGB
   C=VectorParameter() {group=Color Settings, parameterName=ColorC, defaultValue=<rgba 0 0 1 1>}.RGB
   Alpha=MaterialFunctionCall!(
    Texture
    Gradient Start=ScalarParameter() {group=Gradient Settings, parameterName=GradientStart, sortPriority=1}
    Gradient Mid=ScalarParameter() {defaultValue=0.5, group=Gradient Settings, parameterName=GradientMid, sortPriority=2}
    Gradient End=ScalarParameter() {defaultValue=1, group=Gradient Settings, parameterName=GradientEnd, sortPriority=2}
    Rotation Angle=ScalarParameter() {group=Gradient Settings, parameterName=GradientRotation, sortPriority=4}
    Rotation Center=@n1
    Gradient Offset=ScalarParameter() {group=Gradient Settings, parameterName=GradientOffset, sortPriority=3}
   ) {materialFunction=/FortUILibrary/MaterialFunctions/Gradients/MF_UI_LinearGradient_3Colors}.Result
  ) {materialFunction=&Lerp_3Color}
  n2 = ScalarParameter() {group=Gradient Settings, parameterName=GradientOffset, sliderMax=1, sliderMin=-1, sortPriority=2}
  n3 = VectorParameter() {
   group=Rotation Settings
   parameterName=GradientRotationCenter
   sortPriority=1
   channelNames {a=Do Not Use, b=Do Not Use, g=Center Y, r=Center X}
   defaultValue=<rgba 0.5 0.5 0 0>
  }
  n4 = AppendVector(A=@n3.Center X, B=@n3.Center X)
  n5 = ScalarParameter() {group=Rotation Settings, parameterName=GradientRotation, sliderMax=1, sortPriority=0}
  Multiply(
   A=MaterialFunctionCall("Value (0-1)"=MaterialFunctionCall!(UVs, Offset=@n2, RotationCenter=@n4, RotationAngle=@n5) {materialFunction=/FortUILibrary/MaterialFunctions/Gradients/MF_UI_ReflectedGradient}.Result) {materialFunction=&MF_UI_EaseCurves}.EaseInSine
   B=ScalarParameter() {defaultValue=1, parameterName=Opacity, sliderMax=1}
  )
  ScalarParameter() {defaultValue=0.25, group=Gradient Settings, parameterName=GradientWidth, sliderMax=1, sortPriority=0}
  ScalarParameter() {defaultValue=1, group=Gradient Settings, parameterName=GradientDensity, sliderMax=10, sortPriority=1}
  ScalarParameter() {defaultValue=2, group=Gradient Settings, parameterName=GradientPower}
 ]
}
