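# HUD indicator widget for a teammate's reboot card, built on AthenaRebootCardIndicator, with a nameplate showing player marker, name, reboot icon, countdown and attached distance text.
# HandleDisplayForSelf collapses the countdown text for the local player, otherwise shows it, and resets its opacity.
# bp_UpdateCountdownText formats zero-clamped whole seconds, though its inputs appear as constants here.
# An UpdateTime animation transforms and fades the countdown; no graph here plays it.
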WidgetBlueprint /Game/Athena/HUD/Team/RebootCardIndicator
animations [UpdateTime {
  binding_count=1
  length_seconds=0.9384
  object=&self:UpdateTime
  sequence {bindings [TimeText {
     id=#1
     possessed_class=CommonTextBlock
     tracks [
      {
       type=MovieScene2DTransformTrack
       sections [{
         type=MovieScene2DTransformSection
         end_frame=17
         start_frame=-1
         channels [
          Translation.X {type=MovieSceneScriptingFloatChannel}
          Translation.Y {type=MovieSceneScriptingFloatChannel}
          Angle {type=MovieSceneScriptingFloatChannel}
          Shear.X {type=MovieSceneScriptingFloatChannel}
          Shear.Y {type=MovieSceneScriptingFloatChannel}
         ]
        }]
      }
      {
       type=MovieSceneFloatTrack
       display_name=Render Opacity
       sections [{
         type=MovieSceneFloatSection
         end_frame=18
         start_frame=0
         channels [None {type=MovieSceneScriptingFloatChannel, thinned_from=4, keys [[0,1],[10.6583,1],[18.7253,0]]}]
        }]
      }
     ]
    }]}
 }]
graphs [
 EventGraph {nodes [
   n1 = Event HandleDisplayForSelf!() -> (OutputDelegate, then, bSelf:bool) {is=title K2Node_Event}
   n2 = Get TimeText() -> (TimeText:Text Object Reference)
   > SetVisibility(self=@n2, InVisibility=Select(Option 0=HitTestInvisible, Option 1=Collapsed, Index=@n1.bSelf))
   > SetRenderOpacity(self=@n2, InOpacity=1)
  ]}
 bp_UpdateCountdownText {nodes [
   bp_UpdateCountdownText() {is=title K2Node_FunctionEntry}
   n1 = "Max (Integer)"(A=Truncate(A=float - float(A=0, B=1) {is=title K2Node_CallFunction}), B=0)
   > "SetText (Text)"(self=Get TimeText(), InText=Format Text!(Format="LOCTEXT(\"{x}s\")", x:Text="To Text (Integer)"(Value=@n1)) -> (Result))
  ]}
]
variables []
widgets {
 info {inheritedWidgetCount=8, parentClass=/Script/FortniteUI.AthenaRebootCardIndicator, rootWidgetClass=&Overlay, widgetCount=16}
 widgets [
  Overlay_0 {widgetClassPath=&Overlay, widget=Overlay_0}
  Border_Nameplate {
   widgetClassPath=&Border
   bIsVariable=true
   parent=Overlay_0
   widget=Border_Nameplate
   properties {
    background {
     bIsDynamicallyLoaded=false
     drawAs=Box
     imageType=NoImage
     mirroring=NoMirror
     resourceName=None
     resourceObject=&/M_UI_Pill_SquadNamePlate_Inst
     tiling=NoTile
     imageSize=<32 32>
     margin=<ltrb 0.5 0 0.5 0>
     outlineSettings {
      bUseBrushTransparency=false
      roundingType=HalfHeightRadius
      width=0
      color {colorUseRule=UseColor_Specified, specifiedColor=<rgba 0 0 0 0>}
      cornerRadii=<0 0 0 0>
     }
     tintColor {colorUseRule=UseColor_Specified, specifiedColor=<rgba 1 1 1 1>}
     uVRegion {bIsValid=false, max=<0 0>, min=<0 0>}
    }
    brushColor=<rgba 0 0 0 0>
    padding=<ltrb 12 6 15 6>
   }
   slot {class=OverlaySlot}
  }
  HorizontalBox_0 {
   widgetClassPath=&HorizontalBox
   parent=Border_Nameplate
   widget=HorizontalBox_0
   slot {class=BorderSlot, properties {padding=<ltrb 12 6 15 6>}}
  }
  Image_PlayerMarker {
   widgetClassPath=&Image
   bInherited=true
   bIsVariable=true
   parent=HorizontalBox_0
   widget=Image_PlayerMarker
   properties {
    visibility=SelfHitTestInvisible
    brush {
     bIsDynamicallyLoaded=false
     drawAs=Image
     imageType=NoImage
     mirroring=NoMirror
     resourceName=None
     resourceObject=/Game/Athena/HUD/Art/T_UI_Indicator_Diamond
     tiling=NoTile
     imageSize=<13 16>
     margin=<ltrb 0 0 0 0>
     outlineSettings {
      bUseBrushTransparency=false
      roundingType=HalfHeightRadius
      width=0
      color {colorUseRule=UseColor_Specified, specifiedColor=<rgba 0 0 0 0>}
      cornerRadii=<0 0 0 0>
     }
     tintColor {colorUseRule=UseColor_Specified, specifiedColor=<rgba 1 1 1 1>}
     uVRegion {bIsValid=false, max=<0 0>, min=<0 0>}
    }
   }
   slot {class=HorizontalBoxSlot, properties {horizontalAlignment=HAlign_Center, verticalAlignment=VAlign_Center, padding=<ltrb 0 0 4 0>}}
  }
  Text_PlayerName {
   widgetClassPath=&CommonTextBlock
   bInherited=true
   parent=HorizontalBox_0
   widget=Text_PlayerName
   properties {
    text="{MaxCharacterLim}"
    font {
     bForceMonospaced=false
     bMaterialIsStencil=false
     fontMaterial=null
     fontObject=&Burbank
     letterSpacing=0
     monospacedWidth=1
     size=14
     skewAmount=0
     typefaceFontName=BurbankRegular-Black
     outlineSettings {
      bApplyOutlineToDropShadows=true
      bMiteredCorners=false
      bSeparateFillAlpha=false
      outlineMaterial=null
      outlineSize=1
      outlineColor=<rgba 0 0 0 0.75>
     }
    }
    renderTransform {angle=0, scale=<1 1>, shear=<-8 0>, translation=<0 0>}
    shadowColorAndOpacity=<rgba 0 0 0 1>
   }
   slot {class=HorizontalBoxSlot, properties {verticalAlignment=VAlign_Center}}
  }
  IndicatorSpacer {
   widgetClassPath=&Spacer
   bIsVariable=true
   parent=HorizontalBox_0
   widget=IndicatorSpacer
   properties {visibility=Collapsed, size=<6 1>}
   slot {class=HorizontalBoxSlot}
  }
  RebootCountdownContainer {
   widgetClassPath=&HorizontalBox
   bIsVariable=true
   parent=HorizontalBox_0
   widget=RebootCountdownContainer
   slot {class=HorizontalBoxSlot, properties {horizontalAlignment=HAlign_Center, verticalAlignment=VAlign_Center, padding=<ltrb 0 -2 -8 -2>}}
  }
  Image_RebootIcon {
   widgetClassPath=&Image
   bIsVariable=true
   parent=RebootCountdownContainer
   widget=Image_RebootIcon
   properties {brush {
     bIsDynamicallyLoaded=false
     drawAs=Image
     imageType=FullColor
     mirroring=NoMirror
     resourceName=None
     resourceObject=/Game/Athena/HUD/Materials/MI_UI_RebootCardIcon_Indicator
     tiling=NoTile
     imageSize=<32 32>
     margin=<ltrb 0 0 0 0>
     outlineSettings {
      bUseBrushTransparency=false
      roundingType=HalfHeightRadius
      width=0
      color {colorUseRule=UseColor_Specified, specifiedColor=<rgba 0 0 0 0>}
      cornerRadii=<0 0 0 0>
     }
     tintColor {colorUseRule=UseColor_Specified, specifiedColor=<rgba 1 1 1 1>}
     uVRegion {bIsValid=false, max=<0 0>, min=<0 0>}
    }}
   slot {class=HorizontalBoxSlot, properties {horizontalAlignment=HAlign_Right, verticalAlignment=VAlign_Center, padding=<ltrb 6 0 0 0>}}
  }
  TimeText {
   widgetClassPath=&CommonTextBlock
   bInherited=true
   bIsVariable=true
   parent=RebootCountdownContainer
   widget=TimeText
   properties {
    bSimpleTextMode=true
    minDesiredWidth=27.2937
    text=99
    font {
     bForceMonospaced=false
     bMaterialIsStencil=false
     fontMaterial=null
     fontObject=&Burbank
     letterSpacing=0
     monospacedWidth=1
     size=20
     skewAmount=0
     typefaceFontName=Burbank-Black
     outlineSettings {
      bApplyOutlineToDropShadows=false
      bMiteredCorners=false
      bSeparateFillAlpha=false
      outlineMaterial=null
      outlineSize=1
      outlineColor=<rgba 0 0.2183 0.3125 1>
     }
    }
    renderTransform {angle=0, scale=<1 1>, shear=<0 0>, translation=<0 3>}
    shadowColorAndOpacity=<rgba 0.0052 0.0052 0.0056 0.8>
    shadowOffset=<3 3>
   }
   slot {class=HorizontalBoxSlot, properties {horizontalAlignment=HAlign_Left, verticalAlignment=VAlign_Center}}
  }
  Burst {
   widgetClassPath=&Image
   bIsVariable=true
   parent=Overlay_0
   widget=Burst
   properties {
    visibility=Collapsed
    brush {
     bIsDynamicallyLoaded=false
     drawAs=Image
     imageType=NoImage
     mirroring=NoMirror
     resourceName=None
     resourceObject=/Game/UI/Foundation/Textures/Icons/Athena/T_T_Icon_BR_SM_SecondChance_Card_Pickup
     tiling=NoTile
     imageSize=<75 75>
     margin=<ltrb 0 0 0 0>
     outlineSettings {
      bUseBrushTransparency=false
      roundingType=HalfHeightRadius
      width=0
      color {colorUseRule=UseColor_Specified, specifiedColor=<rgba 0 0 0 0>}
      cornerRadii=<0 0 0 0>
     }
     tintColor {colorUseRule=UseColor_Specified, specifiedColor=<rgba 1 1 1 1>}
     uVRegion {bIsValid=false, max=<0 0>, min=<0 0>}
    }
   }
   slot {class=OverlaySlot, properties {horizontalAlignment=HAlign_Center, verticalAlignment=VAlign_Center}}
  }
  AttachText {
   widgetClassPath=&FortVisualAttachment
   parent=Overlay_0
   widget=AttachText
   slot {class=OverlaySlot, properties {horizontalAlignment=HAlign_Right, verticalAlignment=VAlign_Center}}
  }
  Text_Distance {
   widgetClassPath=&CommonTextBlock
   bInherited=true
   parent=AttachText
   widget=Text_Distance
   properties {
    text=100m
    font {
     bForceMonospaced=false
     bMaterialIsStencil=false
     fontMaterial=null
     fontObject=&NotoSans
     letterSpacing=0
     monospacedWidth=1
     size=14
     skewAmount=0
     typefaceFontName=NotoSans-Bold
     outlineSettings {
      bApplyOutlineToDropShadows=false
      bMiteredCorners=false
      bSeparateFillAlpha=false
      outlineMaterial=null
      outlineSize=1
      outlineColor=<rgba 0 0 0 0.6>
     }
    }
    shadowOffset=<0 0>
   }
   slot {class=SizeBoxSlot}
  }
  Switcher_IndicatorSize {widgetClassPath=&WidgetSwitcher, bInherited=true, widget=null}
  Panel_SmallIndicator {widgetClassPath=/Script/UMG.PanelWidget, bInherited=true, widget=null}
  Panel_LargeIndicator {widgetClassPath=&PanelWidget, bInherited=true, widget=null}
  TutorialHighlight {widgetClassPath=&TutorialHighlight}
 ]
}
