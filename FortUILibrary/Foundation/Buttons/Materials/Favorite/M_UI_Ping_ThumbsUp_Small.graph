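# Translucent UI material that draws a small thumbs-up ping icon with a dark outline, tinted by the Icon Color parameter.
# SmoothStep thresholds turn the icon texture's red channel into the fill and its green channel into the outline, blending Icon Color over a near-black background.
# Opacity is the maximum of fill, outline and a shadow sample, whose UV offset uses only ShadowOffset's zero red component, so it overlaps the outline.
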Material /FortUILibrary/Foundation/Buttons/Materials/Favorite/M_UI_Ping_ThumbsUp_Small
graph {
 properties {
  bAutomaticallySetUsageInEditor=false
  bUsedWithStaticMesh=false
  blendMode=BLEND_Translucent
  materialDomain=MD_UI
  usedShadingModels=Default Lit
 }
 nodes [
  n1 = Constant4Vector() {desc=ShadowOffset, constant=<rgba 0 -0.045 0 1>}
  n2 = Subtract(A=Constant() {desc=Edge, r=0.2}, B=Constant() {desc=OutlineSize, r=0.17})
  n3 = Add(A=@n2, B=Constant() {desc=EdgeSoftness, r=0.08})
  n4 = Add(A=TextureCoordinate(), B=AppendVector(A=@n1.R, B=@n1.R))
  n5 = Multiply(
   A=@n1.A
   B=MaterialFunctionCall(Alpha=TextureSample(UVs=@n4) {texture=/FortUILibrary/Foundation/Buttons/Materials/Favorite/Textures/T_UI_Icon_Thumbs_Up}.G, Min=@n2, Max=@n3) {materialFunction=&SmoothStep}.Result
  )
  n6 = TextureSample() {texture=&T_UI_Icon_Thumbs_Up}
  n7 = MaterialFunctionCall(Alpha=@n6.G, Min=@n2, Max=@n3) {materialFunction=&SmoothStep}
  n8 = Constant() {desc=Edge, r=0.2}
  n9 = MaterialFunctionCall(Alpha=@n6.R, Min=@n8, Max=Add(A=@n8, B=Constant() {desc=EdgeSoftness, r=0.08})) {materialFunction=&SmoothStep}
  NamedRerouteDeclaration(None=Max(A=Max(A=@n5, B=@n7.Result), B=@n9.Result)) {name=Opacity, nodeColor=<rgba 0.8281 1 0 1>}
  NamedRerouteDeclaration(None=LinearInterpolate(
    A=Constant3Vector() {desc=BGColor, constant=<rgba 0.0052 0.0052 0.0056 1>}
    B=VectorParameter() {desc=Color, parameterName=Icon Color, defaultValue=<rgba 1 1 1 1>}.RGB
    Alpha=@n9.Result
   )) {name=Final Color, nodeColor=<rgba 0 1 0.5313 1>}
  NamedRerouteUsage()
  NamedRerouteUsage()
 ]
}
