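# Squad marker confirmation HUD widget for Athena, a layout-only Blueprint subclass of the native AthenaMarkerConfirmationWidget.
# A canvas root holds the inherited ConfirmationIconParent horizontal box, auto-sized and centered by anchors and alignment.
# The box's render pivot sits at bottom center, so any render transforms act from that point.
# The event graph has only unconnected event stubs, leaving all behavior to the parent class.
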WidgetBlueprint /Game/Athena/HUD/Team/SquadMarkerConfirmationWidget
graphs [EventGraph {nodes [
   Event PreConstruct()
   Event Construct()
   Event Tick()
  ]}]
variables []
widgets {
 info {
  inheritedWidgetCount=1
  parentClass=/Script/FortniteUI.AthenaMarkerConfirmationWidget
  rootWidgetClass=&CanvasPanel
  widgetCount=2
 }
 widgets [
  CanvasPanel_216 {widgetClassPath=&CanvasPanel, widget=CanvasPanel_216, properties {renderTransformPivot=<0 0>}}
  ConfirmationIconParent {
   widgetClassPath=&HorizontalBox
   bInherited=true
   bIsVariable=true
   parent=CanvasPanel_216
   widget=ConfirmationIconParent
   properties {renderTransformPivot=<0.5 1>}
   slot {
    class=CanvasPanelSlot
    properties {bAutoSize=true, layoutData {alignment=<0.5 0.5>, anchors {maximum=<0.5 0.5>, minimum=<0.5 0.5>}, offsets=<ltrb 0 0 0 0>}}
   }
  }
 ]
}
