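# SquadMarker is the HUD widget for squad ping markers, a child of AthenaSquadIndicator with an empty event graph.
# It stacks a thumbs-up confirmation widget above overlaid small and large indicators, a burst image and distance text.
# Anim_DangerPlaced plays a red burst through material parameters, while Anim_Hovered morphs, shrinks and fades the large indicator.
# The LocationPlaced and ItemPlaced animations have no bindings.
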WidgetBlueprint /Game/Athena/HUD/Team/SquadMarker
animations [
 Anim_DangerPlaced {
  binding_count=1
  length_seconds=0.651
  object=&self:Anim_DangerPlaced
  sequence {bindings [Burst {
     id=#1
     possessed_class=Image
     tracks [
      {
       type=MovieSceneWidgetMaterialTrack
       sections [{
         type=MovieSceneParameterSection
         end_frame=8
         start_frame=-5
         channels [
          Alpha {type=MovieSceneScriptingFloatChannel, thinned_from=7, keys [[-0.949,0],[-0.5157,0],[0.4843,1],[3.484,1],[6,0.1967],[7.152,0]]}
          Grow {type=MovieSceneScriptingFloatChannel, thinned_from=4, keys [[-0.949,0],[-0.5157,0],[3.4843,1]]}
          RadialDensity {type=MovieSceneScriptingFloatChannel, thinned_from=5, keys [[-0.949,0],[-0.5157,0],[1.484,2],[7.1517,0.25]]}
          Radius {type=MovieSceneScriptingFloatChannel, thinned_from=6, keys [[-0.949,0],[-0.516,0],[1.484,0.4],[7.1517,0.75],[7.152,1]]}
          RingAlpha {type=MovieSceneScriptingFloatChannel, thinned_from=4, keys [[-0.949,0],[-0.516,0],[5.484,1]]}
          RingSize {type=MovieSceneScriptingFloatChannel, thinned_from=4, keys [[-0.949,0],[0.484,0],[5.484,1]]}
          RingThicc {type=MovieSceneScriptingFloatChannel, thinned_from=4, keys [[-0.949,0],[0.484,0],[5.484,1]]}
          Flare {type=MovieSceneScriptingFloatChannel, thinned_from=6, keys [[-0.949,0],[-0.568,0],[1.1507,1],[7.1517,0.75],[7.152,0]]}
          FlareAlpha {type=MovieSceneScriptingFloatChannel, thinned_from=6, keys [[-0.949,0],[-0.5677,0],[1.484,1],[3.4843,1],[7.152,0]]}
          Color_2R {type=MovieSceneScriptingFloatChannel, const [1,2]}
          Color_2G {type=MovieSceneScriptingFloatChannel, const [0.0081,2]}
          Color_2B {type=MovieSceneScriptingFloatChannel, const [0,2]}
          Color_2A {type=MovieSceneScriptingFloatChannel, const [0,2]}
          Color_1R {type=MovieSceneScriptingFloatChannel, keys [[0,1],[6,0.62]]}
          Color_1G {type=MovieSceneScriptingFloatChannel, keys [[0,0.0904],[6,0]]}
          Color_1B {type=MovieSceneScriptingFloatChannel, keys [[0,0.2911],[6,0.02]]}
          Color_1A {type=MovieSceneScriptingFloatChannel, const [0,2]}
         ]
        }]
      }
      {
       type=MovieSceneByteTrack
       display_name=Visibility
       sections [{type=MovieSceneByteSection, end_frame=8, start_frame=0, channels [None {type=MovieSceneScriptingByteChannel, keys [[0,4],[8.8543,2]]}]}]
      }
     ]
    }]}
 }
 Anim_LocationPlaced {binding_count=0, length_seconds=0.5089, object=&self:Anim_LocationPlaced, sequence {bindings []}}
 Anim_ItemPlaced {binding_count=0, length_seconds=0.5089, object=&self:Anim_ItemPlaced, sequence {bindings []}}
 Anim_Hovered {
  binding_count=1
  length_seconds=0.2
  object=&self:Anim_Hovered
  sequence {bindings [Item_Large {
     id=#2
     display_name=Image_LargeIndicator
     possessed_class=Image
     tracks [
      {
       type=MovieSceneWidgetMaterialTrack
       sections [{
         type=MovieSceneParameterSection
         end_frame=4
         start_frame=0
         channels [
          SolidFull {type=MovieSceneScriptingFloatChannel, keys [[0,0],[4,1]]}
          IndicatorShape {type=MovieSceneScriptingFloatChannel, keys [[0,0],[4,1]]}
         ]
        }]
      }
      {
       type=MovieScene2DTransformTrack
       sections [{
         type=MovieScene2DTransformSection
         end_frame=4
         start_frame=0
         channels [
          Translation.X {type=MovieSceneScriptingFloatChannel}
          Translation.Y {type=MovieSceneScriptingFloatChannel}
          Angle {type=MovieSceneScriptingFloatChannel}
          Scale.X {type=MovieSceneScriptingFloatChannel, keys [[0,0.8],[3,0.3318],[4,0.3318]]}
          Scale.Y {type=MovieSceneScriptingFloatChannel, keys [[0,0.8],[3,0.3318],[4,0.3318]]}
          Shear.X {type=MovieSceneScriptingFloatChannel}
          Shear.Y {type=MovieSceneScriptingFloatChannel}
         ]
        }]
      }
      {
       type=MovieSceneFloatTrack
       display_name=Render Opacity
       sections [{type=MovieSceneFloatSection, end_frame=4, start_frame=0, channels [None {type=MovieSceneScriptingFloatChannel, keys [[0,1],[4,0.5]]}]}]
      }
     ]
    }]}
 }
]
graphs [EventGraph {links [], nodes []}]
variables []
widgets {
 info {inheritedWidgetCount=11, parentClass=/Script/FortniteUI.AthenaSquadIndicator, rootWidgetClass=&VerticalBox, widgetCount=16}
 widgets [
  VerticalBox_0 {widgetClassPath=&VerticalBox, widget=VerticalBox_0}
  FortVisualAttachment_0 {
   widgetClassPath=&FortVisualAttachment
   parent=VerticalBox_0
   widget=FortVisualAttachment_0
   properties {contentAnchor=<0.5 0>, renderTransform {angle=0, scale=<0.65 0.65>, shear=<0 0>, translation=<0 -14>}, renderTransformPivot=<0.5 1>}
   slot {class=VerticalBoxSlot, properties {horizontalAlignment=HAlign_Center, verticalAlignment=VAlign_Top}}
  }
  MarkerConfirmationWidget {
   widgetClassPath=&/SquadMarkerConfirmationWidget
   bInherited=true
   bIsVariable=true
   parent=FortVisualAttachment_0
   widget=MarkerConfirmationWidget
   properties {
    confirmationIconAsset=/FortUILibrary/Foundation/Buttons/Materials/Favorite/M_UI_Ping_ThumbsUp_Small
    renderOpacity=0.88
    renderTransformPivot=<0.5 0>
   }
   slot {class=SizeBoxSlot, properties {horizontalAlignment=HAlign_Center, verticalAlignment=VAlign_Top, padding=<ltrb 10 0 0 0>}}
  }
  Overlay_0 {widgetClassPath=&Overlay, parent=VerticalBox_0, widget=Overlay_0, slot {class=VerticalBoxSlot}}
  Panel_SmallIndicator {
   widgetClassPath=&Overlay
   bInherited=true
   parent=Overlay_0
   widget=Panel_SmallIndicator
   slot {class=OverlaySlot, properties {horizontalAlignment=HAlign_Fill, verticalAlignment=VAlign_Fill}}
  }
  Image_TrajectorySmallIndicator {
   widgetClassPath=&Image
   bInherited=true
   parent=Panel_SmallIndicator
   widget=Image_TrajectorySmallIndicator
   properties {
    visibility=Collapsed
    brush {
     bIsDynamicallyLoaded=false
     drawAs=Image
     imageType=NoImage
     mirroring=NoMirror
     resourceName=None
     resourceObject=/Game/Athena/HUD/Materials/M_UI_HexSocket_Inst
     tiling=NoTile
     imageSize=<32 32>
     margin=<ltrb 0 0 0 0>
     outlineSettings {
      bUseBrushTransparency=false
      roundingType=HalfHeightRadius
      width=0
      color {colorUseRule=UseColor_Specified, specifiedColor=<rgba 0 0 0 0>}
      cornerRadii=<0 0 0 0>
     }
     tintColor {colorUseRule=UseColor_Specified, specifiedColor=<rgba 1 1 1 1>}
     uVRegion {bIsValid=false, max=<0 0>, min=<0 0>}
    }
   }
   slot {class=OverlaySlot, properties {horizontalAlignment=HAlign_Center, verticalAlignment=VAlign_Bottom}}
  }
  Image_SmallIndicator {
   widgetClassPath=&Image
   bInherited=true
   parent=Panel_SmallIndicator
   widget=Image_SmallIndicator
   properties {
    brush {
     bIsDynamicallyLoaded=false
     drawAs=Image
     imageType=NoImage
     mirroring=NoMirror
     resourceName=None
     resourceObject=/Game/Athena/HUD/Materials/M_PingIndicator
     tiling=NoTile
     imageSize=<16 16>
     margin=<ltrb 0 0 0 0>
     outlineSettings {
      bUseBrushTransparency=false
      roundingType=HalfHeightRadius
      width=0
      color {colorUseRule=UseColor_Specified, specifiedColor=<rgba 0 0 0 0>}
      cornerRadii=<0 0 0 0>
     }
     tintColor {colorUseRule=UseColor_Specified, specifiedColor=<rgba 1 1 1 1>}
     uVRegion {bIsValid=false, max=<0 0>, min=<0 0>}
    }
    renderTransform {angle=0, scale=<1 1>, shear=<0 0>, translation=<0 -3>}
   }
   slot {class=OverlaySlot, properties {horizontalAlignment=HAlign_Center, verticalAlignment=VAlign_Bottom}}
  }
  Panel_LargeIndicator {
   widgetClassPath=&Overlay
   bInherited=true
   parent=Overlay_0
   widget=Panel_LargeIndicator
   properties {renderTransform {angle=0, scale=<0.8 0.8>, shear=<0 0>, translation=<0 0>}}
   slot {class=OverlaySlot, properties {horizontalAlignment=HAlign_Fill, verticalAlignment=VAlign_Fill}}
  }
  Image_TrajectoryLargeIndicator {
   widgetClassPath=&Image
   bInherited=true
   parent=Panel_LargeIndicator
   widget=Image_TrajectoryLargeIndicator
   properties {
    visibility=Collapsed
    brush {
     bIsDynamicallyLoaded=false
     drawAs=Image
     imageType=NoImage
     mirroring=NoMirror
     resourceName=None
     resourceObject=&M_UI_HexSocket_Inst
     tiling=NoTile
     imageSize=<45 45>
     margin=<ltrb 0 0 0 0>
     outlineSettings {
      bUseBrushTransparency=false
      roundingType=HalfHeightRadius
      width=0
      color {colorUseRule=UseColor_Specified, specifiedColor=<rgba 0 0 0 0>}
      cornerRadii=<0 0 0 0>
     }
     tintColor {colorUseRule=UseColor_Specified, specifiedColor=<rgba 1 1 1 1>}
     uVRegion {bIsValid=false, max=<0 0>, min=<0 0>}
    }
   }
   slot {class=OverlaySlot}
  }
  Image_LocationFlare {
   widgetClassPath=&Image
   bInherited=true
   parent=Panel_LargeIndicator
   widget=Image_LocationFlare
   properties {
    visibility=SelfHitTestInvisible
    brush {
     bIsDynamicallyLoaded=false
     drawAs=Image
     imageType=NoImage
     mirroring=NoMirror
     resourceName=None
     resourceObject=/Game/Athena/HUD/Materials/M_UI_LocationPingReminder
     tiling=NoTile
     imageSize=<25 25>
     margin=<ltrb 0 0 0 0>
     outlineSettings {
      bUseBrushTransparency=false
      roundingType=HalfHeightRadius
      width=0
      color {colorUseRule=UseColor_Specified, specifiedColor=<rgba 0 0 0 0>}
      cornerRadii=<0 0 0 0>
     }
     tintColor {colorUseRule=UseColor_Specified, specifiedColor=<rgba 1 1 1 1>}
     uVRegion {bIsValid=false, max=<0 0>, min=<0 0>}
    }
    renderTransform {angle=0, scale=<3 3>, shear=<0 0>, translation=<0 6>}
   }
   slot {class=OverlaySlot, properties {horizontalAlignment=HAlign_Center}}
  }
  Image_LargeIndicator {
   widgetClassPath=&Image
   bInherited=true
   parent=Panel_LargeIndicator
   widget=Image_LargeIndicator
   properties {brush {
     bIsDynamicallyLoaded=false
     drawAs=Image
     imageType=NoImage
     mirroring=NoMirror
     resourceName=None
     resourceObject=/Game/Athena/HUD/Materials/M_UI_Ping_Location_Test
     tiling=NoTile
     imageSize=<45 45>
     margin=<ltrb 0 0 0 0>
     outlineSettings {
      bUseBrushTransparency=false
      roundingType=HalfHeightRadius
      width=0
      color {colorUseRule=UseColor_Specified, specifiedColor=<rgba 0 0 0 0>}
      cornerRadii=<0 0 0 0>
     }
     tintColor {colorUseRule=UseColor_Specified, specifiedColor=<rgba 1 1 1 1>}
     uVRegion {bIsValid=false, max=<0 0>, min=<0 0>}
    }}
   slot {class=OverlaySlot}
  }
  Burst {
   widgetClassPath=&Image
   parent=Overlay_0
   widget=Burst
   properties {
    visibility=Hidden
    brush {
     bIsDynamicallyLoaded=false
     drawAs=Image
     imageType=NoImage
     mirroring=NoMirror
     resourceName=None
     resourceObject=/Game/Athena/HUD/Art/squadArt/M_UI_Burst_Ping
     tiling=NoTile
     imageSize=<32 32>
     margin=<ltrb 0 0 0 0>
     outlineSettings {
      bUseBrushTransparency=false
      roundingType=HalfHeightRadius
      width=0
      color {colorUseRule=UseColor_Specified, specifiedColor=<rgba 0 0 0 0>}
      cornerRadii=<0 0 0 0>
     }
     tintColor {colorUseRule=UseColor_Specified, specifiedColor=<rgba 1 1 1 1>}
     uVRegion {bIsValid=false, max=<0 0>, min=<0 0>}
    }
    renderTransform {angle=0, scale=<7 7>, shear=<0 0>, translation=<0 0>}
   }
   slot {class=OverlaySlot, properties {horizontalAlignment=HAlign_Center, verticalAlignment=VAlign_Center}}
  }
  AttachText {
   widgetClassPath=&FortVisualAttachment
   parent=Overlay_0
   widget=AttachText
   properties {renderTransform {angle=0, scale=<1 1>, shear=<0 0>, translation=<-10 0>}}
   slot {class=OverlaySlot, properties {horizontalAlignment=HAlign_Right, verticalAlignment=VAlign_Center}}
  }
  Text_Distance {
   widgetClassPath=&CommonTextBlock
   bInherited=true
   parent=AttachText
   widget=Text_Distance
   properties {
    text=100m
    font {
     bForceMonospaced=false
     bMaterialIsStencil=false
     fontMaterial=null
     fontObject=&Burbank
     letterSpacing=0
     monospacedWidth=1
     size=14
     skewAmount=0
     typefaceFontName=BurbankRegular-Black
     outlineSettings {
      bApplyOutlineToDropShadows=false
      bMiteredCorners=false
      bSeparateFillAlpha=false
      outlineMaterial=&M-FontShadow
      outlineSize=3
      outlineColor=<rgba 0 0 0 0.6>
     }
    }
    renderTransform {angle=0, scale=<1 1>, shear=<-8 0>, translation=<0 0>}
    shadowColorAndOpacity=<rgba 0 0 0 0.8>
    shadowOffset=<4 3>
   }
   slot {class=SizeBoxSlot}
  }
  Switcher_IndicatorSize {widgetClassPath=&WidgetSwitcher, bInherited=true, widget=null}
  TutorialHighlight {widgetClassPath=&TutorialHighlight}
 ]
}
