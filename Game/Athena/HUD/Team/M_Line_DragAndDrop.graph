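# Translucent UI material for a drag-and-drop line in the team HUD, tinted by the PingColor parameter.
# Two triangle-pattern textures pan at different speeds and combine with a horizontal UV gradient into an animated effect, which multiplies the tint.
# A separate mask multiplies the inverted second-UV-channel gradient by an antialiased SDF box, then applies a power curve to shape the falloff.
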Material /Game/Athena/HUD/Team/M_Line_DragAndDrop
graph {
 properties {blendMode=BLEND_Translucent, materialDomain=MD_UI, pixelDepthOffsetMode=PDOM_Legacy, usedShadingModels=Default Lit}
 nodes [
  n1 = TextureCoordinate()
  n2 = ComponentMask(None=@n1) {r=true}
  n3 = TextureSample(UVs=Panner(Coordinate=@n1) {speedX=0.3, speedY=0.1}) {samplerSource=SSM_Wrap_WorldGroupSettings, texture=/Game/Global/Textures/Default/Masks_and_Gradients/T_TrianglePattern}
  n4 = TextureSample(UVs=Multiply(A=Panner(Coordinate=@n1) {speedX=0.8, speedY=0.08}, B=Constant() {r=0.75}) {constB=0.75}) {
   samplerSource=SSM_Wrap_WorldGroupSettings
   samplerType=SAMPLERTYPE_LinearColor
   texture=/Game/UI/Foundation/Textures/Pattern/T_Triangle_BattlePass
  }
  n5 = Add(A=Multiply(A=Subtract(A=@n2) {constB=0.15}, B=@n3.R), B=@n4.G)
  n6 = Multiply(A=@n4.B)
  n7 = Add(A=Multiply(A=Multiply(A=@n5), B=@n4.G), B=Add(A=Multiply(A=@n6, B=@n3.R), B=Multiply(A=@n3.G) {constB=5}))
  NamedRerouteDeclaration(None=Add(A=Multiply(A=@n7) {constB=2}, B=ComponentMask(None=@n1) {r=true})) {name=VFX, nodeColor=<rgba 0 1 0.5547 1>}
  n8 = ComponentMask(None=OneMinus(None=TextureCoordinate() {coordinateIndex=1})) {r=true}
  n9 = Constant() {r=1}
  n10 = OneMinus(None=MaterialFunctionCall(DistanceField=MaterialFunctionCall!(UVs=@n8, Size=@n9, CornerRoundness, KeepAspectRatio, PixelSize) {materialFunction=/FortUILibrary/MaterialFunctions/Shapes/SDFs/MF_SDF_Box}.SDF) {materialFunction=&MF_UI_AntiAliasing}.Outer)
  Saturate(None=Power(Base=Multiply(A=@n8, B=@n10)) {constExponent=0.3})
  Saturate(None=Multiply(A=VectorParameter() {parameterName=PingColor, defaultValue=<rgba 0.0911 0.5679 0.9896 1>}.RGB, B=NamedRerouteUsage()))
 ]
}
