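# Widget Blueprint for the Athena HUD's drag-and-drop marker details, based on the native AthenaMarkerDragAndDropDetails class.
# A transparent border and overlay hold the inherited MarkerPointer, which draws a square and takes its color through the PingColor material parameter.
# On PreConstruct it creates a dynamic instance of M_Line_DragAndDrop, stores it as LineMaterial and assigns it as the pointer's line mesh material.
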WidgetBlueprint /Game/Athena/HUD/Team/WBP_AthenaMarkerDragAndDropDetails
graphs [EventGraph {nodes [
   Event PreConstruct()
   > n1 = CreateDynamicMaterialInstance!(execute, Parent=&/M_Line_DragAndDrop, OptionalName=None, CreationFlags=None) -> (then, ReturnValue)
   > n2 = Set LineMaterial(LineMaterial:Material Instance Dynamic Object Reference=@n1) -> (Output_Get:Material Instance Dynamic Object Reference)
   > SetLineMeshMaterial!(
    execute
    self:Athena Marker Pointer Object Reference=Get MarkerPointer() -> (MarkerPointer:Athena Marker Pointer Object Reference)
    Material:Material Interface Object Reference=@n2
   ) -> (then)
  ]}]
variables [LineMaterial {type=MaterialInstanceDynamic}]
widgets {
 info {inheritedWidgetCount=2, parentClass=/Script/FortniteUI.AthenaMarkerDragAndDropDetails, rootWidgetClass=&Border, widgetCount=4}
 widgets [
  ContentBox {widgetClassPath=&Border, bIsVariable=true, widget=ContentBox, properties {brushColor=<rgba 0 0 0 0>, padding=<ltrb 0 0 0 0>}}
  Overlay_37 {widgetClassPath=&Overlay, parent=ContentBox, widget=Overlay_37, slot {class=BorderSlot, properties {padding=<ltrb 0 0 0 0>}}}
  MarkerPointer {
   widgetClassPath=/Script/FortniteUI.AthenaMarkerPointer
   bInherited=true
   bIsVariable=true
   parent=Overlay_37
   widget=MarkerPointer
   properties {bDrawSquare=true, materialColorParamName=PingColor}
   slot {class=OverlaySlot, properties {verticalAlignment=VAlign_Center}}
  }
  TutorialHighlight {widgetClassPath=&TutorialHighlight}
 ]
}
