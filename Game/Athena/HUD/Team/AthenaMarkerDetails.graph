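# Hover details widget for world markers in the team HUD, showing a marker's name, colored icon bubble, pointer line and action prompt.
# UpdateMarkerData caches display info from a details table, plays the hover animation, then switches on marker type to pick an icon and tint dynamic materials.
# Unloaded icons load asynchronously, with an equality check against the current icon before finishing setup.
# An optional async-created custom details widget replaces the default layout, and a tag-matched host slot accepts extension content.
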WidgetBlueprint /Game/Athena/HUD/Team/AthenaMarkerDetails
animations [OnHover {
  binding_count=2
  length_seconds=0.4
  object=&self:OnHover
  sequence {bindings [
    HorizontalBox_0 {
     id=#1
     possessed_class=HorizontalBox
     tracks [
      {
       type=MovieSceneFloatTrack
       display_name=Render Opacity
       sections [{type=MovieSceneFloatSection, end_frame=8, start_frame=0, channels [None {type=MovieSceneScriptingFloatChannel, keys [[0,0],[7,1],[8,1]]}]}]
      }
      {
       type=MovieScene2DTransformTrack
       sections [{
         type=MovieScene2DTransformSection
         end_frame=7
         start_frame=0
         channels [
          Translation.X {type=MovieSceneScriptingFloatChannel, keys [[0,-20],[4,0],[7,0]]}
          Translation.Y {type=MovieSceneScriptingFloatChannel}
          Angle {type=MovieSceneScriptingFloatChannel}
          Scale.X {type=MovieSceneScriptingFloatChannel}
          Scale.Y {type=MovieSceneScriptingFloatChannel}
          Shear.X {type=MovieSceneScriptingFloatChannel}
          Shear.Y {type=MovieSceneScriptingFloatChannel}
         ]
        }]
      }
     ]
    }
    Seperator {
     id=#2
     possessed_class=Border
     tracks [{
       type=MovieScene2DTransformTrack
       sections [{
         type=MovieScene2DTransformSection
         end_frame=8
         start_frame=0
         channels [
          Translation.X {type=MovieSceneScriptingFloatChannel}
          Translation.Y {type=MovieSceneScriptingFloatChannel}
          Angle {type=MovieSceneScriptingFloatChannel}
          Scale.X {type=MovieSceneScriptingFloatChannel, keys [[0,0.293],[6.6353,1],[8,1]]}
          Scale.Y {type=MovieSceneScriptingFloatChannel, const [1,2]}
          Shear.X {type=MovieSceneScriptingFloatChannel}
          Shear.Y {type=MovieSceneScriptingFloatChannel}
         ]
        }]
      }]
    }
   ]}
 }]
graphs [
 EventGraph {nodes [
   TryHandleIconLoad() {is=title K2Node_CustomEvent}
   n1 = Break Marked Actor Display Info!(MarkedActorDisplayInfo:"Marked Actor Display Info Structure (by ref)"=Get CachedDisplayInfo() -> (CachedDisplayInfo:Marked Actor Display Info Structure)) -> (
    DisplayName:Text
    Icon:Texture 2D Soft Object Reference
    CustomIndicatorClass:User Widget Soft Class Reference
    CustomDetailsClass:User Widget Soft Class Reference
    PrimaryColor:Linear Color Structure
    Sound:Sound Base Object Reference
    bAttachSound:bool=~
    ScreenClamping:EFortMarkedActorScreenClamping Enum
    SpecialCull:Bitmask Integer
    SpecialDotSizeThreshold:float=~
    SpecialDistance:float=~
    CustomTimeToLive:float=~
    AdditionalData:Marked Actor Display Info Additional Data Object Reference
   ) {is=title K2Node_BreakStruct}
   > n2 = Is Valid(InputObject=Resolve Soft Reference!(Input:Texture 2D Soft Object Reference=@n1.Icon) -> (Output:Texture 2D Object Reference))
   n3 = Async Load Asset(execute=@n2.Is Not Valid, Asset=@n1.Icon)
   n4 = Break Marked Actor Display Info(MarkedActorDisplayInfo=Get CachedDisplayInfo()) -> (bAttachSound=~, SpecialDotSizeThreshold=~, SpecialDistance=~, CustomTimeToLive=~)
   Branch(execute=@n3.Completed, Condition="Equal (Object)"(A=@n3, B=Resolve Soft Reference(Input=@n4.Icon)) {is=title K2Node_CallFunction})
   > FinishMarkedCustomIconSetup()
   FinishMarkedCustomIconSetup(execute=@n2.Is Valid)
   FinishMarkedCustomIconSetup() {is=title K2Node_CustomEvent}
   n5 = Break Marked Actor Display Info(MarkedActorDisplayInfo=Get CachedDisplayInfo()) -> (bAttachSound=~, SpecialDotSizeThreshold=~, SpecialDistance=~, CustomTimeToLive=~)
   > SetDisplayText!(execute, self, NewText:Text=@n5) -> (then)
   > SetBrushColor(self=Get Seperator() -> (Seperator:Border Object Reference), InBrushColor=@n5.PrimaryColor)
   > n6 = Async Load Asset(Asset=@n5.Icon)
   n7 = GetIconByMarkerType!(execute=@n6.Completed, self) -> (then, ImageObj:Image Object Reference)
   > n8 = GetDynamicMaterial(self=@n7)
   > n9 = Cast To Texture2D(Object=@n6)
   > SetTextureParameterValue(self=@n8, ParameterName=DisplayedIcon, Value=@n9)
   > SetScalarParameterValue(self=@n8, ParameterName=UseIconColor, Value=1)
   > SetScalarParameterValue(self=@n8, ParameterName=IconColorStrength, Value=0.94)
   n10 = BreakColor(InColor=@n5.PrimaryColor)
   n11 = Get ColorReduction() -> (ColorReduction:double)
   > SetVectorParameterValue(
    self=@n8
    ParameterName=FillColor
    Value=MakeColor!(R:float=float * float(A=@n10, B=@n11), G:float=float * float(A=@n10.G, B=@n11), B:float=float * float(A=@n10.B, B=@n11), A:float=0.9) -> (ReturnValue:Linear Color Structure)
   )
   > SetVectorParameterValue(self=@n8, ParameterName=FillColorInner, Value=@n5.PrimaryColor)
   > SetVectorParameterValue(self=@n8, ParameterName=RingColor, Value=@n5.PrimaryColor)
   > SetVectorParameterValue(self=@n8, ParameterName=IconColor, Value=(R=1,G=1,B=1,A=1))
   > SetVectorParameterValue(self=@n8, ParameterName=IconScale, Value=MakeColor(R=0.85, G=0.85, B=0, A=1))
   > SetVectorParameterValue(self=Get LineMaterial() -> (LineMaterial:Material Instance Dynamic Object Reference), ParameterName=LineColor, Value=@n5.PrimaryColor)
   Check Marker Overrides() {is=title K2Node_CustomEvent}
   > SetActiveWidgetIndex(self=Get OverrideSwitcher() -> (OverrideSwitcher:Widget Switcher Object Reference), Index=0)
   n12 = Break Marked Actor Display Info(MarkedActorDisplayInfo=Get CachedDisplayInfo()) -> (bAttachSound=~, SpecialDotSizeThreshold=~, SpecialDistance=~, CustomTimeToLive=~)
   > Branch(Condition=IsValidSoftClassReference!(SoftClassReference:"Object Soft Class Reference (by ref)"=@n12.CustomDetailsClass) -> (ReturnValue:bool))
   > n13 = CreateWidgetAsync!(
    execute:Exec
    UserWidgetSoftClass:User Widget Soft Class Reference=@n12.CustomDetailsClass
    OwningPlayer:Player Controller Object Reference
    bSuspendInputUntilComplete:bool=true
   ) -> (then:Exec, OnComplete:Exec, OnFailure:Exec, UserWidget:User Widget Object Reference) {is=title K2Node_AsyncAction}
   Is Valid(exec=@n13.OnComplete, InputObject=@n13)
   n14 = Get OverrideSlot() -> (OverrideSlot:Named Slot Object Reference)
   > ClearChildren(self=@n14)
   > AddChild(self=@n14, Content=@n13)
   > SetActiveWidgetIndex(self=Get OverrideSwitcher(), Index=1)
   > n15 = Cast To FortActorMarkerInterface!(execute, Object=@n13) -> (then, CastFailed, AsFort Actor Marker Interface:Fort Actor Marker Interface Interface) {is=title K2Node_DynamicCast}
   > SetMarkerData!(
    execute
    self:Fort Actor Marker Interface Interface=@n15
    MarkerData:"Fort World Marker Data Structure (by ref)"=Get MyMarkerData() -> (MyMarkerData:Fort World Marker Data Structure)
    DisplayInfo:"Marked Actor Display Info Structure (by ref)"=Get CachedDisplayInfo()
   ) -> (then)
   SetupLocationMarker() {is=title K2Node_CustomEvent}
   > SetRenderTranslation(self=Get MarkerPointer() -> (MarkerPointer:Athena Marker Pointer Object Reference), Translation=(X=-7,Y=5))
   n16 = Break Marked Actor Display Info(MarkedActorDisplayInfo=Get CachedDisplayInfo()) -> (bAttachSound=~, SpecialDotSizeThreshold=~, SpecialDistance=~, CustomTimeToLive=~)
   > SetDisplayText(NewText=@n16)
   > n17 = GetDynamicMaterial(self=Get LocationIcon() -> (LocationIcon:Image Object Reference))
   > SetBrushColor(self=Get Seperator(), InBrushColor=@n16.PrimaryColor)
   > SetScalarParameterValue(self=@n17, ParameterName=UseIconColor, Value=0)
   > SetVectorParameterValue(self=@n17, ParameterName=Color, Value=@n16.PrimaryColor)
   > SetVectorParameterValue(self=@n17, ParameterName=RingColor, Value=@n16.PrimaryColor)
   > SetTextureParameterValue(self=@n17, ParameterName=DisplayedIcon, Value=/Game/Athena/HUD/Art/T_UI_Indicator_Diamond)
   > SetVectorParameterValue(self=@n17, ParameterName=IconScale, Value=MakeColor(R=0.2238, G=0.3, B=0, A=1))
   > SetVectorParameterValue(self=Get LineMaterial(), ParameterName=LineColor, Value=@n16.PrimaryColor)
   SetupEnemyMarker() {is=title K2Node_CustomEvent}
   > SetRenderTranslation(self=Get MarkerPointer(), Translation=(X=-49,Y=19))
   n18 = Break Marked Actor Display Info(MarkedActorDisplayInfo=Get CachedDisplayInfo()) -> (bAttachSound=~, SpecialDotSizeThreshold=~, SpecialDistance=~, CustomTimeToLive=~)
   > SetDisplayText(NewText=@n18)
   > n19 = GetDynamicMaterial(self=Get DangerIcon() -> (DangerIcon:Image Object Reference))
   > SetBrushColor(self=Get Seperator(), InBrushColor=@n18.PrimaryColor)
   > SetScalarParameterValue(self=@n19, ParameterName=UseIconColor, Value=0)
   > SetVectorParameterValue(self=@n19, ParameterName=IconColor, Value=@n18.PrimaryColor)
   > SetVectorParameterValue(self=@n19, ParameterName=RingColor, Value=@n18.PrimaryColor)
   > SetTextureParameterValue(self=@n19, ParameterName=DisplayedIcon, Value=/Game/UI/Foundation/Textures/Icons/Replay/T-Icon-Replay-Exclamation-L)
   > SetVectorParameterValue(self=@n19, ParameterName=IconScale, Value=MakeColor(R=0.5, G=0.5, B=0, A=1))
   > SetVectorParameterValue(self=Get LineMaterial(), ParameterName=LineColor, Value=@n18.PrimaryColor)
   TryHandleNPCCommandMarker() {is=title K2Node_CustomEvent}
   n20 = Get CachedDisplayInfo()
   > n21 = Is Valid(InputObject=Resolve Soft Reference(Input=Break Marked Actor Display Info(MarkedActorDisplayInfo=@n20) -> (bAttachSound=~, SpecialDotSizeThreshold=~, SpecialDistance=~, CustomTimeToLive=~).Icon))
   n22 = Async Load Asset(execute=@n21.Is Not Valid)
   n23 = Break Marked Actor Display Info(MarkedActorDisplayInfo=Get CachedDisplayInfo()) -> (bAttachSound=~, SpecialDotSizeThreshold=~, SpecialDistance=~, CustomTimeToLive=~)
   Branch(execute=@n22.Completed, Condition="Equal (Object)"(A=@n22, B=Resolve Soft Reference(Input=@n23.Icon)))
   > FinishNPCCommandMarker()
   FinishNPCCommandMarker(execute=@n21.Is Valid)
   TryHandleSquadComsMarker() {is=title K2Node_CustomEvent}
   n24 = Get CachedDisplayInfo()
   > n25 = Is Valid(InputObject=Resolve Soft Reference(Input=Break Marked Actor Display Info(MarkedActorDisplayInfo=@n24) -> (bAttachSound=~, SpecialDotSizeThreshold=~, SpecialDistance=~, CustomTimeToLive=~).Icon))
   n26 = Async Load Asset(execute=@n25.Is Not Valid)
   n27 = Break Marked Actor Display Info(MarkedActorDisplayInfo=Get CachedDisplayInfo()) -> (bAttachSound=~, SpecialDotSizeThreshold=~, SpecialDistance=~, CustomTimeToLive=~)
   Branch(execute=@n26.Completed, Condition="Equal (Object)"(A=@n26, B=Resolve Soft Reference(Input=@n27.Icon)))
   > FinishSquadComsMarker()
   FinishSquadComsMarker(execute=@n25.Is Valid)
   FinishNPCCommandMarker() {is=title K2Node_CustomEvent}
   n28 = Break Marked Actor Display Info(MarkedActorDisplayInfo=Get CachedDisplayInfo()) -> (bAttachSound=~, SpecialDotSizeThreshold=~, SpecialDistance=~, CustomTimeToLive=~)
   > SetDisplayText(NewText=@n28)
   > SetBrushColor(self=Get Seperator(), InBrushColor=@n28.PrimaryColor)
   > n29 = GetDynamicMaterial(self=Get SquadComsIcon() -> (SquadComsIcon:Image Object Reference))
   > SetScalarParameterValue(self=@n29, ParameterName=bCropIconToCircle, Value=1)
   > SetScalarParameterValue(self=@n29, ParameterName=UseIconColor, Value=1)
   > SetScalarParameterValue(self=@n29, ParameterName=IconColorStrength, Value=0.94)
   > SetTextureParameterValue(self=@n29, ParameterName=DisplayedIcon, Value=Resolve Soft Reference(Input=@n28.Icon))
   n30 = BreakColor(InColor=@n28.PrimaryColor)
   n31 = Get ColorReduction()
   > SetVectorParameterValue(
    self=@n29
    ParameterName=FillColor
    Value=MakeColor(R=float * float(A=@n30, B=@n31), G=float * float(A=@n30.G, B=@n31), B=float * float(A=@n30.B, B=@n31), A=0.9)
   )
   > SetVectorParameterValue(self=@n29, ParameterName=FillColorInner, Value=@n28.PrimaryColor)
   > SetVectorParameterValue(self=@n29, ParameterName=RingColor, Value=@n28.PrimaryColor)
   > SetVectorParameterValue(self=@n29, ParameterName=IconColor, Value=(R=1,G=1,B=1,A=1))
   > SetVectorParameterValue(self=@n29, ParameterName=IconScale, Value=MakeColor(R=0.85, G=0.85, B=0, A=1))
   > SetVectorParameterValue(self=Get LineMaterial(), ParameterName=LineColor, Value=@n28.PrimaryColor)
   FinishSquadComsMarker() {is=title K2Node_CustomEvent}
   n32 = Break Marked Actor Display Info(MarkedActorDisplayInfo=Get CachedDisplayInfo()) -> (bAttachSound=~, SpecialDotSizeThreshold=~, SpecialDistance=~, CustomTimeToLive=~)
   > SetDisplayText(NewText=@n32)
   > SetBrushColor(self=Get Seperator(), InBrushColor=@n32.PrimaryColor)
   > n33 = GetDynamicMaterial(self=Get SquadComsIcon())
   > SetScalarParameterValue(self=@n33, ParameterName=bCropIconToCircle, Value=0)
   > SetScalarParameterValue(self=@n33, ParameterName=UseIconColor, Value=1)
   > SetScalarParameterValue(self=@n33, ParameterName=IconColorStrength, Value=0.94)
   > SetTextureParameterValue(self=@n33, ParameterName=DisplayedIcon, Value=Resolve Soft Reference(Input=@n32.Icon))
   n34 = BreakColor(InColor=@n32.PrimaryColor)
   n35 = Get ColorReduction()
   > SetVectorParameterValue(
    self=@n33
    ParameterName=FillColor
    Value=MakeColor(R=float * float(A=@n34, B=@n35), G=float * float(A=@n34.G, B=@n35), B=float * float(A=@n34.B, B=@n35), A=0.9)
   )
   > SetVectorParameterValue(self=@n33, ParameterName=FillColorInner, Value=@n32.PrimaryColor)
   > SetVectorParameterValue(self=@n33, ParameterName=RingColor, Value=@n32.PrimaryColor)
   > SetVectorParameterValue(self=@n33, ParameterName=IconColor, Value=(R=1,G=1,B=1,A=1))
   > SetVectorParameterValue(self=@n33, ParameterName=IconScale, Value=MakeColor(R=0.85, G=0.85, B=0, A=1))
   > SetVectorParameterValue(self=Get LineMaterial(), ParameterName=LineColor, Value=@n32.PrimaryColor)
   Event PreConstruct()
   > n36 = CreateDynamicMaterialInstance!(execute, Parent=&/M_Line, OptionalName=None, CreationFlags=None) -> (then, ReturnValue)
   > n37 = Set LineMaterial(LineMaterial:Material Instance Dynamic Object Reference=@n36) -> (Output_Get:Material Instance Dynamic Object Reference)
   > SetLineMeshMaterial!(execute, self:Athena Marker Pointer Object Reference=Get MarkerPointer(), Material:Material Interface Object Reference=@n37) -> (then)
   n38 = Event UpdateMarkerAction!() -> (OutputDelegate, then, MarkerAction:EFortMarkerActions Enum=None) {is=title K2Node_Event}
   n39 = GetVisibility!(self:Widget Object Reference) -> (ReturnValue:ESlateVisibility Enum=Visible)
   > Branch(Condition=NOT Boolean(A="Equal (Enum)"(A:ESlateVisibility Enum=@n39, B:ESlateVisibility Enum=Hidden)))
   > SetVisibility(
    self=Get ActionTextContainer() -> (ActionTextContainer:Horizontal Box Object Reference)
    InVisibility=Select!(
     None:ESlateVisibility Enum=Hidden
     Cancel:ESlateVisibility Enum=SelfHitTestInvisible
     Confirm:ESlateVisibility Enum=SelfHitTestInvisible
     CancelAndReturnNPC:ESlateVisibility Enum=SelfHitTestInvisible
     Unconfirm:ESlateVisibility Enum=SelfHitTestInvisible
     Index:EFortMarkerActions Enum=@n38.MarkerAction
    ) -> (ReturnValue)
   )
   > "SetText (Text)"(
    self=Get ActionText() -> (ActionText:Common Text Object Reference)
    InText=Select(
     None:Text
     Cancel:Text="LOCTEXT(\"Cancel Marker\")"
     Confirm:Text="LOCTEXT(\"Confirm\")"
     CancelAndReturnNPC:Text="LOCTEXT(\"Cancel && Return\")"
     Unconfirm:Text="LOCTEXT(\"Unconfirm\")"
     Index=@n38.MarkerAction
    ) -> (ReturnValue:Text)
   )
   Event Construct()
   > Show!(execute, self, Show:bool=false) -> (then)
   n40 = Get MarkerDetailsWidgetSlot() -> (MarkerDetailsWidgetSlot:Gameplay Tag Structure)
   > RegisterWidgetHost!(execute, WidgetHost:Widget Object Reference, WidgetHostTag:"Gameplay Tag Structure (by ref)"=@n40, IsPlayerSpecific:bool=true) -> (then)
   n41 = Event UpdateMarkerData!() -> (OutputDelegate, then, MarkerData:"Fort World Marker Data Structure (by ref)") {is=title K2Node_Event}
   n42 = Break Fort World Marker Data!(FortWorldMarkerData:"Fort World Marker Data Structure (by ref)"=@n41.MarkerData) -> (
    MarkerID:Marker ID Structure
    Owner:Fort Player State Weak object
    MarkerType:EFortWorldMarkerType Enum
    BasePositionOffset:Vector=~
    WorldNormal:Vector=~
    ItemDefinition:Item Definition Base Object Reference
    ItemCount:int
    PinnedQuestIndex:int
    MarkedActor:Actor Weak object
    MarkedActorClass:Actor Soft Class Reference
    MarkerTag:Gameplay Tag Structure
    bUseHoveredMarkerDetail:bool=~
   ) {is=title K2Node_BreakStruct}
   > Show(Show=@n42.bUseHoveredMarkerDetail)
   > Branch(Condition=@n42.bUseHoveredMarkerDetail)
   > Set MyMarkerData(MyMarkerData:Fort World Marker Data Structure=@n41.MarkerData) -> (Output_Get:Fort World Marker Data Structure)
   > n43 = GetMarkedItemDetails!(execute, self) -> (then, DisplayInfo:Marked Actor Display Info Structure)
   > Set CachedDisplayInfo(CachedDisplayInfo:Marked Actor Display Info Structure=@n43) -> (Output_Get:Marked Actor Display Info Structure)
   > SetActiveWidgetIndex(self=Get DetailSwitcher() -> (DetailSwitcher:Widget Switcher Object Reference), Index=0)
   > SetVisibility(self=Get ContentBox() -> (ContentBox:Border Object Reference), InVisibility=SelfHitTestInvisible)
   > PlayAnimation(InAnimation=Get OnHover() -> (OnHover:Widget Animation Object Reference), PlaybackSpeed=2)
   > n44 = Sequence()
   > n45 = Switch on EFortWorldMarkerType!(execute, Selection:EFortWorldMarkerType Enum=@n42.MarkerType) -> (
    None:Exec
    Location:Exec
    Enemy:Exec
    Item:Exec
    Elimination:Exec
    SpecialLocal:Exec
    SpecialServer:Exec
    Teammate:Exec
    PingGoTo:Exec
    PingHoldPosition:Exec
    PingNoIcon:Exec
    PingAttack:Exec
    SquadComsWatchArea:Exec
    SquadComsDefend:Exec
    SquadComsSomeonesHere:Exec
    SquadComsLootArea:Exec
    SquadComsAttack:Exec
    SquadComsFallBack:Exec
    SquadRequestsAmmoLight:Exec
    SquadRequestsAmmoMedium:Exec
    SquadRequestsAmmoHeavy:Exec
    SquadRequestsAmmoShells:Exec
    SquadRequestsAmmoExplosives:Exec
    SquadRequestsMaterialWood:Exec
    SquadRequestsMaterialStone:Exec
    SquadRequestsMaterialMetal:Exec
    SquadRequestsHealth:Exec
    SquadRequestsShields:Exec
    SquadRequestsGold:Exec
    SquadComsHelp:Exec
    SquadComsPickMeUp:Exec
    SquadComsThankYou:Exec
    SquadComsReady:Exec
    SquadComsFollowMe:Exec
    SquadComsHoldOn:Exec
   ) {is=title K2Node_SwitchEnum}
   SetupEnemyMarker(execute=@n45.Enemy) {is=title K2Node_CallFunction}
   > SetActiveWidgetIndex(self=Get Switcher_IndicatorIcon() -> (Switcher_IndicatorIcon:Widget Switcher Object Reference), Index=2)
   SetRenderTranslation(execute=[@n45.Item, @n45.SpecialServer], self=Get MarkerPointer(), Translation=(X=-49,Y=19))
   > TryHandleIconLoad() {is=title K2Node_CallFunction}
   > SetActiveWidgetIndex(self=Get Switcher_IndicatorIcon(), Index=0)
   SetupLocationMarker(execute=@n45.Location) {is=title K2Node_CallFunction}
   > SetActiveWidgetIndex(self=Get Switcher_IndicatorIcon(), Index=1)
   SetRenderTranslation(execute=[@n45.PingGoTo, @n45.PingHoldPosition], Translation=(X=-49,Y=19))
   > TryHandleNPCCommandMarker() {is=title K2Node_CallFunction}
   > SetActiveWidgetIndex(self=Get Switcher_IndicatorIcon(), Index=4)
   SetRenderTranslation(execute=@n45.SpecialLocal, self=Get MarkerPointer(), Translation=(X=-5,Y=5))
   > TryHandleIconLoad()
   > SetActiveWidgetIndex(self=Get Switcher_IndicatorIcon(), Index=3)
   > SetActiveWidgetIndex(self=Get DetailSwitcher(), Index=1)
   SetRenderTranslation(
    execute=[
     @n45.SquadComsAttack
     @n45.SquadComsDefend
     @n45.SquadComsFallBack
     @n45.SquadComsHelp
     @n45.SquadComsLootArea
     @n45.SquadComsPickMeUp
     @n45.SquadComsSomeonesHere
     @n45.SquadComsThankYou
     @n45.SquadComsWatchArea
     @n45.SquadRequestsAmmoExplosives
     @n45.SquadRequestsAmmoHeavy
     @n45.SquadRequestsAmmoLight
     @n45.SquadRequestsAmmoMedium
     @n45.SquadRequestsAmmoShells
     @n45.SquadRequestsGold
     @n45.SquadRequestsHealth
     @n45.SquadRequestsMaterialMetal
     @n45.SquadRequestsMaterialStone
     @n45.SquadRequestsMaterialWood
     @n45.SquadRequestsShields
    ]
    Translation=(X=-49,Y=19)
   )
   > TryHandleSquadComsMarker() {is=title K2Node_CallFunction}
   > SetActiveWidgetIndex(self=Get Switcher_IndicatorIcon(), Index=4)
   TryHandleIconLoad(execute=@n45.Teammate)
   > SetActiveWidgetIndex(self=Get Switcher_IndicatorIcon(), Index=5)
   Check Marker Overrides(execute=@n44.then_1) {is=title K2Node_CallFunction}
   n46 = Event UpdateMarkerConfirmations!() -> (OutputDelegate, then, MarkerData:"Fort World Marker Data Structure (by ref)") {is=title K2Node_Event}
   > ShowMarkerConfirmations!(
    execute
    self:Athena Marker Confirmation Widget Object Reference=Get SquadMarkerConfirmationWidget() -> (SquadMarkerConfirmationWidget:Squad Marker Confirmation Widget Object Reference)
    InMarkerData:"Fort World Marker Data Structure (by ref)"=@n46.MarkerData
   ) -> (then)
   n47 = Event Add Content To Host Widget!() -> (
    OutputDelegate
    then
    WidgetHostTag:"Gameplay Tag Structure (by ref)"
    ContentWidget:Widget Object Reference
    Padding:"Margin Structure (by ref)"
    bFill:bool
   ) {is=title K2Node_Event}
   > Branch(Condition="Equal (GameplayTag)"(A=@n47.WidgetHostTag, B=Get MarkerDetailsWidgetSlot()))
   > n48 = AddChildToVerticalBox(
    self=Get VerticalBox_ExtensionWidgetHolder() -> (VerticalBox_ExtensionWidgetHolder:Vertical Box Object Reference)
    Content=@n47.ContentWidget
   )
   > SetHorizontalAlignment(self:Vertical Box Slot Object Reference=@n48, InHorizontalAlignment=HAlign_Fill)
   > SetVerticalAlignment(self:Vertical Box Slot Object Reference=@n48, InVerticalAlignment=VAlign_Fill)
   Event Destruct()
   > UnregisterWidgetHost!(execute, WidgetHost:Widget Object Reference, WidgetHostTag:"Gameplay Tag Structure (by ref)"=@n40) -> (then)
  ]}
 GetIconByMarkerType {nodes [
   GetIconByMarkerType() {is=title K2Node_FunctionEntry}
   n1 = Get ItemIcon() -> (ItemIcon:Image Object Reference)
   > Return Node!(
    execute
    ImageObj:Image Object Reference=Select!(
     None:Image Object Reference
     Location:Image Object Reference
     Enemy:Image Object Reference
     Item:Image Object Reference=@n1
     Elimination:Image Object Reference
     SpecialLocal:Image Object Reference
     SpecialServer:Image Object Reference=@n1
     Teammate:Image Object Reference=Get TeammateIcon() -> (TeammateIcon:Image Object Reference)
     PingGoTo:Image Object Reference
     PingHoldPosition:Image Object Reference
     PingNoIcon:Image Object Reference
     PingAttack:Image Object Reference
     SquadComsWatchArea:Image Object Reference
     SquadComsDefend:Image Object Reference
     SquadComsSomeonesHere:Image Object Reference
     SquadComsLootArea:Image Object Reference
     SquadComsAttack:Image Object Reference
     SquadComsFallBack:Image Object Reference
     SquadRequestsAmmoLight:Image Object Reference
     SquadRequestsAmmoMedium:Image Object Reference
     SquadRequestsAmmoHeavy:Image Object Reference
     SquadRequestsAmmoShells:Image Object Reference
     SquadRequestsAmmoExplosives:Image Object Reference
     SquadRequestsMaterialWood:Image Object Reference
     SquadRequestsMaterialStone:Image Object Reference
     SquadRequestsMaterialMetal:Image Object Reference
     SquadRequestsHealth:Image Object Reference
     SquadRequestsShields:Image Object Reference
     SquadRequestsGold:Image Object Reference
     SquadComsHelp:Image Object Reference
     SquadComsPickMeUp:Image Object Reference
     SquadComsThankYou:Image Object Reference
     SquadComsReady:Image Object Reference
     SquadComsFollowMe:Image Object Reference
     SquadComsHoldOn:Image Object Reference
     Index:EFortWorldMarkerType Enum=Get!() -> (
      MyMarkerData_MarkerID:Marker ID Structure
      MyMarkerData_Owner:Fort Player State Object Reference
      MyMarkerData_MarkerType:EFortWorldMarkerType Enum
      MyMarkerData_BasePositionOffset:Vector=~
      MyMarkerData_WorldNormal:Vector=~
      MyMarkerData_ItemDefinition:Item Definition Base Object Reference
      MyMarkerData_ItemCount:int
      MyMarkerData_PinnedQuestIndex:int
      MyMarkerData_MarkedActor:Actor Object Reference
      MyMarkerData_MarkedActorClass:Actor Soft Class Reference
      MyMarkerData_MarkerTag:Gameplay Tag Structure
      MyMarkerData_bUseHoveredMarkerDetail:bool=~
     ).MyMarkerData_MarkerType
    ) -> (ReturnValue:Image Object Reference)
   )
  ]}
 GetMarkedItemDetails {nodes [
   GetMarkedItemDetails() {is=title K2Node_FunctionEntry}
   > n1 = GetDisplayInfoForMarkerData!(
    execute
    self:Fort Marker Details Table Object Reference=Get DetailsTable() -> (DetailsTable:Fort Marker Details Table Object Reference)
    MarkerData:"Fort World Marker Data Structure (by ref)"=Get MyMarkerData()
   ) -> (then, ReturnValue:Marked Actor Display Info Structure)
   > Return Node!(execute, DisplayInfo:Marked Actor Display Info Structure=@n1)
  ]}
 RemoveContentFromWidgetHost {nodes [
   n1 = Remove Content From Host Widget!() -> (
    then
    WidgetHostTag:"Gameplay Tag Structure (by ref)"
    ContentWidget:Widget Object Reference
    RemovalType:EWidgetSlottingRemovalType Enum=Unparent
   ) {is=title K2Node_FunctionEntry}
   > Branch(Condition="Equal (GameplayTag)"(A=@n1, B=Get MarkerDetailsWidgetSlot()))
   > n2 = RemoveChild(self=Get VerticalBox_ExtensionWidgetHolder(), Content=Get ContentWidget() -> (ContentWidget:Widget Object Reference))
   > Return Node!(execute, ReturnValue:bool=@n2)
  ]}
 SetDisplayText {nodes [
   n1 = SetDisplayText!() -> (then, NewText:Text) {is=title K2Node_FunctionEntry}
   n2 = Get ItemName() -> (ItemName:Text Object Reference)
   > "SetText (Text)"(self=@n2, InText=TextToUpper!(InText:"Text (by ref)"=@n1) -> (ReturnValue:Text))
   > SetVisibility(self=@n2, InVisibility=HitTestInvisible)
  ]}
 SetLineColor {nodes [
   n1 = SetLineColor!() -> (then, NewColor:Linear Color Structure) {is=title K2Node_FunctionEntry}
   > SetVectorParameterValue(self=Get LineMaterial(), ParameterName=LineColor, Value=@n1)
  ]}
 Show {nodes [
   n1 = Show!() -> (then, Show:bool=~) {is=title K2Node_FunctionEntry}
   > n2 = Branch(Condition=@n1)
   > SetVisibility(InVisibility=SelfHitTestInvisible)
   > SetVisibility(self=[Get ItemName(), Get SquadComsIcon(), Get Seperator(), Get LocationIcon(), Get ItemIcon()], InVisibility=HitTestInvisible)
   > SetVisibility(self=Get MarkerPointer(), InVisibility=HitTestInvisible)
   SetVisibility(execute=@n2.else, InVisibility=Hidden)
   > SetVisibility(
    self=[Get Seperator(), Get SquadComsIcon(), Get ItemName(), Get MarkerPointer(), Get LocationIcon(), Get ActionTextContainer(), Get ItemIcon()]
    InVisibility=Hidden
   )
  ]}
]
variables [
 CachedDisplayInfo {
  type=MarkedActorDisplayInfo
  default {
   additionalData=null
   bAttachSound=true
   customDetailsClass=null
   customIndicatorClass=null
   customTimeToLive=-1
   displayName=""
   icon=null
   screenClamping=Default
   sound=null
   specialCull=0
   specialDistance=50
   specialDotSizeThreshold=-0.93
   primaryColor=<rgba 0 0 0 0>
  }
 }
 ColorReduction {type=number, default=0.08}
 LineMaterial {type=MaterialInstanceDynamic}
 MarkerDetailsWidgetSlot {type=GameplayTag, default {tagName=DynamicUI.HUD.Fort.Athena.Marker.MarkerDetailsSlot}}
 MyMarkerData {
  type=FortWorldMarkerData
  default {
   bHeldMarkerInput=false
   bIncludeSquad=true
   bUseHoveredMarkerDetail=true
   itemCount=1
   itemDefinition=null
   markedActor=null
   markedActorClass=null
   markerType=None
   mostRecentArrayReplicationKey=-1
   owner=null
   pingCommand=Invalid
   pinnedQuestIndex=-1
   replicationId=-1
   replicationKey=-1
   targetConvertedType=Unknown
   basePosition=<0 0 0>
   basePositionOffset=<0 0 0>
   confirmationContainer {confirmations []}
   customDisplayInfos []
   markerId {instanceId=-1, playerId=-1}
   markerTag {tagName=None}
   worldNormal=<0 0 1>
  }
 }
]
widgets {
 info {inheritedWidgetCount=3, parentClass=/Script/FortniteUI.AthenaMarkerDetails, rootWidgetClass=&Border, widgetCount=30}
 widgets [
  ContentBox {widgetClassPath=&Border, bIsVariable=true, widget=ContentBox, properties {brushColor=<rgba 0 0 0 0>, padding=<ltrb 0 0 0 0>}}
  Overlay_3 {widgetClassPath=&Overlay, parent=ContentBox, widget=Overlay_3, slot {class=BorderSlot, properties {padding=<ltrb 0 0 0 0>}}}
  MarkerPointer {
   widgetClassPath=/Script/FortniteUI.AthenaMarkerPointer
   bInherited=true
   bIsVariable=true
   parent=Overlay_3
   widget=MarkerPointer
   properties {renderTransform {angle=0, scale=<1 1>, shear=<0 0>, translation=<-7 5>}}
   slot {class=OverlaySlot, properties {verticalAlignment=VAlign_Center}}
  }
  OverrideSwitcher {widgetClassPath=&WidgetSwitcher, bIsVariable=true, parent=Overlay_3, widget=OverrideSwitcher, slot {class=OverlaySlot}}
  VerticalBox_1 {widgetClassPath=&VerticalBox, parent=OverrideSwitcher, widget=VerticalBox_1, slot {class=WidgetSwitcherSlot}}
  VerticalBox {widgetClassPath=&VerticalBox, parent=VerticalBox_1, widget=VerticalBox, slot {class=VerticalBoxSlot}}
  HorizontalBox_0 {
   widgetClassPath=&HorizontalBox
   parent=VerticalBox
   widget=HorizontalBox_0
   slot {class=VerticalBoxSlot, properties {horizontalAlignment=HAlign_Left, padding=<ltrb 6 0 8 -6>}}
  }
  ItemName {
   widgetClassPath=&TextBlock
   bIsVariable=true
   parent=HorizontalBox_0
   widget=ItemName
   properties {
    text="{MARKER TITLE}"
    font {
     bForceMonospaced=false
     bMaterialIsStencil=false
     fontMaterial=null
     fontObject=&Burbank
     letterSpacing=0
     monospacedWidth=1
     size=20
     skewAmount=0
     typefaceFontName=Burbank-PlayerPower
     outlineSettings {
      bApplyOutlineToDropShadows=true
      bMiteredCorners=false
      bSeparateFillAlpha=true
      outlineMaterial=&M-FontShadow
      outlineSize=3
      outlineColor=<rgba 0 0 0 0.8>
     }
    }
    renderTransform {angle=0, scale=<1 1>, shear=<-8 0>, translation=<0 0>}
   }
   slot {class=HorizontalBoxSlot, properties {horizontalAlignment=HAlign_Left, verticalAlignment=VAlign_Bottom, padding=<ltrb 8 -5 0 -5>}}
  }
  Overlay_0 {widgetClassPath=&Overlay, parent=VerticalBox, widget=Overlay_0, slot {class=VerticalBoxSlot, properties {padding=<ltrb 0 4 0 0>}}}
  Seperator {
   widgetClassPath=&Border
   bIsVariable=true
   parent=Overlay_0
   widget=Seperator
   properties {
    background {
     bIsDynamicallyLoaded=false
     drawAs=Image
     imageType=NoImage
     mirroring=NoMirror
     resourceName=None
     resourceObject=null
     tiling=NoTile
     imageSize=<32 16>
     margin=<ltrb 0 0 0 0>
     outlineSettings {
      bUseBrushTransparency=false
      roundingType=HalfHeightRadius
      width=0
      color {colorUseRule=UseColor_Specified, specifiedColor=<rgba 0 0 0 0>}
      cornerRadii=<0 0 0 0>
     }
     tintColor {colorUseRule=UseColor_Specified, specifiedColor=<rgba 1 1 1 1>}
     uVRegion {bIsValid=false, max=<0 0>, min=<0 0>}
    }
    brushColor=<rgba 1 0 0 1>
    padding=<ltrb 4 0 4 3>
    renderTransformPivot=<0 0.5>
   }
   slot {class=OverlaySlot, properties {horizontalAlignment=HAlign_Fill, verticalAlignment=VAlign_Center, padding=<ltrb 3 0 0 0>}}
  }
  FortVisualAttachment_0 {
   widgetClassPath=&FortVisualAttachment
   parent=Overlay_0
   widget=FortVisualAttachment_0
   properties {widthOverride=100}
   slot {class=OverlaySlot, properties {verticalAlignment=VAlign_Center}}
  }
  Overlay_1 {
   widgetClassPath=&Overlay
   parent=FortVisualAttachment_0
   widget=Overlay_1
   slot {class=SizeBoxSlot, properties {horizontalAlignment=HAlign_Right, verticalAlignment=VAlign_Top}}
  }
  Switcher_IndicatorIcon {
   widgetClassPath=&WidgetSwitcher
   bIsVariable=true
   parent=Overlay_1
   widget=Switcher_IndicatorIcon
   properties {activeWidgetIndex=1}
   slot {class=OverlaySlot, properties {horizontalAlignment=HAlign_Fill, verticalAlignment=VAlign_Fill}}
  }
  ItemIcon {
   widgetClassPath=&Image
   bIsVariable=true
   parent=Switcher_IndicatorIcon
   widget=ItemIcon
   properties {
    brush {
     bIsDynamicallyLoaded=false
     drawAs=Image
     imageType=NoImage
     mirroring=NoMirror
     resourceName=None
     resourceObject=/Game/Athena/HUD/Materials/M_UI_PingBubble_LargeIItemPing
     tiling=NoTile
     imageSize=<75 75>
     margin=<ltrb 0 0 0 0>
     outlineSettings {
      bUseBrushTransparency=false
      roundingType=HalfHeightRadius
      width=0
      color {colorUseRule=UseColor_Specified, specifiedColor=<rgba 0 0 0 0>}
      cornerRadii=<0 0 0 0>
     }
     tintColor {colorUseRule=UseColor_Specified, specifiedColor=<rgba 1 1 1 1>}
     uVRegion {bIsValid=false, max=<0 0>, min=<0 0>}
    }
    renderTransform {angle=0, scale=<1 1>, shear=<0 0>, translation=<-65 -37>}
    renderTransformPivot=<1 0.5>
   }
   slot {class=WidgetSwitcherSlot}
  }
  LocationIcon {
   widgetClassPath=&Image
   bIsVariable=true
   parent=Switcher_IndicatorIcon
   widget=LocationIcon
   properties {
    brush {
     bIsDynamicallyLoaded=false
     drawAs=Image
     imageType=NoImage
     mirroring=NoMirror
     resourceName=None
     resourceObject=/Game/Athena/HUD/Materials/M_UI_Ping_Location_Test
     tiling=NoTile
     imageSize=<40 40>
     margin=<ltrb 0 0 0 0>
     outlineSettings {
      bUseBrushTransparency=false
      roundingType=HalfHeightRadius
      width=0
      color {colorUseRule=UseColor_Specified, specifiedColor=<rgba 0 0 0 0>}
      cornerRadii=<0 0 0 0>
     }
     tintColor {colorUseRule=UseColor_Specified, specifiedColor=<rgba 1 1 1 1>}
     uVRegion {bIsValid=false, max=<0 0>, min=<0 0>}
    }
    renderTransform {angle=0, scale=<1 1>, shear=<0 0>, translation=<-23 -17>}
    renderTransformPivot=<1 0.5>
   }
   slot {class=WidgetSwitcherSlot}
  }
  DangerIcon {
   widgetClassPath=&Image
   bIsVariable=true
   parent=Switcher_IndicatorIcon
   widget=DangerIcon
   properties {
    brush {
     bIsDynamicallyLoaded=false
     drawAs=Image
     imageType=NoImage
     mirroring=NoMirror
     resourceName=None
     resourceObject=/Game/Athena/HUD/Materials/M_UI_PingBubble_Inst
     tiling=NoTile
     imageSize=<75 75>
     margin=<ltrb 0 0 0 0>
     outlineSettings {
      bUseBrushTransparency=false
      roundingType=HalfHeightRadius
      width=0
      color {colorUseRule=UseColor_Specified, specifiedColor=<rgba 0 0 0 0>}
      cornerRadii=<0 0 0 0>
     }
     tintColor {colorUseRule=UseColor_Specified, specifiedColor=<rgba 1 1 1 1>}
     uVRegion {bIsValid=false, max=<0 0>, min=<0 0>}
    }
    renderTransform {angle=0, scale=<1 1>, shear=<0 0>, translation=<-65 -37>}
    renderTransformPivot=<1 0.5>
   }
   slot {class=WidgetSwitcherSlot}
  }
  Reboot {
   widgetClassPath=&Image
   bIsVariable=true
   parent=Switcher_IndicatorIcon
   widget=Reboot
   properties {
    brush {
     bIsDynamicallyLoaded=false
     drawAs=Image
     imageType=NoImage
     mirroring=NoMirror
     resourceName=None
     resourceObject=/Game/Athena/HUD/Materials/M_UI_RebootIcon_Vis_Inst
     tiling=NoTile
     imageSize=<64 64>
     margin=<ltrb 0 0 0 0>
     outlineSettings {
      bUseBrushTransparency=false
      roundingType=HalfHeightRadius
      width=0
      color {colorUseRule=UseColor_Specified, specifiedColor=<rgba 0 0 0 0>}
      cornerRadii=<0 0 0 0>
     }
     tintColor {colorUseRule=UseColor_Specified, specifiedColor=<rgba 1 1 1 1>}
     uVRegion {bIsValid=false, max=<0 0>, min=<0 0>}
    }
    colorAndOpacity=<rgba 0 0.8714 1 1>
    renderTransform {angle=0, scale=<1 1>, shear=<0 0>, translation=<-51 -42>}
   }
   slot {class=WidgetSwitcherSlot}
  }
  SquadComsIcon {
   widgetClassPath=&Image
   bIsVariable=true
   parent=Switcher_IndicatorIcon
   widget=SquadComsIcon
   properties {
    brush {
     bIsDynamicallyLoaded=false
     drawAs=Image
     imageType=NoImage
     mirroring=NoMirror
     resourceName=None
     resourceObject=&M_UI_PingBubble_LargeIItemPing
     tiling=NoTile
     imageSize=<75 75>
     margin=<ltrb 0 0 0 0>
     outlineSettings {
      bUseBrushTransparency=false
      roundingType=HalfHeightRadius
      width=0
      color {colorUseRule=UseColor_Specified, specifiedColor=<rgba 0 0 0 0>}
      cornerRadii=<0 0 0 0>
     }
     tintColor {colorUseRule=UseColor_Specified, specifiedColor=<rgba 1 1 1 1>}
     uVRegion {bIsValid=false, max=<0 0>, min=<0 0>}
    }
    renderTransform {angle=0, scale=<1 1>, shear=<0 0>, translation=<-65 -37>}
    renderTransformPivot=<1 0.5>
   }
   slot {class=WidgetSwitcherSlot}
  }
  TeammateIcon {
   widgetClassPath=&Image
   bIsVariable=true
   parent=Switcher_IndicatorIcon
   widget=TeammateIcon
   properties {
    brush {
     bIsDynamicallyLoaded=false
     drawAs=Image
     imageType=NoImage
     mirroring=NoMirror
     resourceName=None
     resourceObject=/Game/Athena/HUD/Materials/M_UI_PingBubble_LargeTeammatePing
     tiling=NoTile
     imageSize=<75 75>
     margin=<ltrb 0 0 0 0>
     outlineSettings {
      bUseBrushTransparency=false
      roundingType=HalfHeightRadius
      width=0
      color {colorUseRule=UseColor_Specified, specifiedColor=<rgba 0 0 0 0>}
      cornerRadii=<0 0 0 0>
     }
     tintColor {colorUseRule=UseColor_Specified, specifiedColor=<rgba 1 1 1 1>}
     uVRegion {bIsValid=false, max=<0 0>, min=<0 0>}
    }
    renderTransform {angle=0, scale=<1 1>, shear=<0 0>, translation=<-65 -37>}
    renderTransformPivot=<1 0.5>
   }
   slot {class=WidgetSwitcherSlot}
  }
  VerticalBox_ExtensionWidgetHolder {
   widgetClassPath=&VerticalBox
   bInherited=true
   bIsVariable=true
   parent=VerticalBox
   widget=VerticalBox_ExtensionWidgetHolder
   slot {class=VerticalBoxSlot}
  }
  DetailSwitcher {
   widgetClassPath=&WidgetSwitcher
   bIsVariable=true
   parent=VerticalBox
   widget=DetailSwitcher
   slot {class=VerticalBoxSlot, properties {horizontalAlignment=HAlign_Left, padding=<ltrb 14 2 8 0>}}
  }
  ActionTextContainer {
   widgetClassPath=&HorizontalBox
   bIsVariable=true
   parent=DetailSwitcher
   widget=ActionTextContainer
   slot {class=WidgetSwitcherSlot}
  }
  ScaleBox_0 {
   widgetClassPath=&ScaleBox
   parent=ActionTextContainer
   widget=ScaleBox_0
   properties {stretch=UserSpecified, userSpecifiedScale=0.6}
   slot {class=HorizontalBoxSlot, properties {verticalAlignment=VAlign_Center}}
  }
  KeybindWidget {
   widgetClassPath=/FortUILibrary/Input/KeybindWidget
   bIsVariable=true
   parent=ScaleBox_0
   widget=KeybindWidget
   properties {boundAction=PlaceMapMarkerInWorld}
   slot {class=ScaleBoxSlot}
  }
  ActionText {
   widgetClassPath=&CommonTextBlock
   bIsVariable=true
   parent=ActionTextContainer
   widget=ActionText
   properties {
    mobileFontSizeMultiplier=1.25
    text=Text Block
    textTransformPolicy=ToUpper
    font {
     bForceMonospaced=false
     bMaterialIsStencil=false
     fontMaterial=null
     fontObject=&BurbankSmall
     letterSpacing=0
     monospacedWidth=1
     size=14
     skewAmount=0
     typefaceFontName=Bold
     outlineSettings {
      bApplyOutlineToDropShadows=true
      bMiteredCorners=false
      bSeparateFillAlpha=true
      outlineMaterial=&M-FontShadow
      outlineSize=3
      outlineColor=<rgba 0 0 0 0.5>
     }
    }
    shadowOffset=<0 0>
   }
   slot {class=HorizontalBoxSlot, properties {verticalAlignment=VAlign_Center, padding=<ltrb 4 -2 0 -2>}}
  }
  TextBlock_RebootCard {
   widgetClassPath=&CommonRichTextBlock
   bIsVariable=true
   parent=DetailSwitcher
   widget=TextBlock_RebootCard
   properties {
    bOverrideDefaultStyle=true
    lineHeightPercentage=0.8
    mobileTextBlockScale=1.3
    text=Bring Reboot Cards to a Van to Restore Squadmates
    wrapTextAt=260
    defaultTextStyleOverride {
     overflowPolicy=Clip
     transformPolicy=None
     colorAndOpacity {colorUseRule=UseColor_Specified, specifiedColor=<rgba 1 1 1 1>}
     font {
      bForceMonospaced=false
      bMaterialIsStencil=false
      fontMaterial=null
      fontObject=&BurbankSmall
      letterSpacing=0
      monospacedWidth=1
      size=14
      skewAmount=0
      typefaceFontName=Bold
      outlineSettings {
       bApplyOutlineToDropShadows=true
       bMiteredCorners=false
       bSeparateFillAlpha=true
       outlineMaterial=&M-FontShadow
       outlineSize=3
       outlineColor=<rgba 0 0 0 0.5>
      }
     }
     highlightColor {colorUseRule=UseColor_Specified, specifiedColor=<rgba 0 0 0 1>}
     highlightShape {
      bIsDynamicallyLoaded=false
      drawAs=Image
      imageType=NoImage
      mirroring=NoMirror
      resourceName=None
      resourceObject=null
      tiling=NoTile
      imageSize=<32 32>
      margin=<ltrb 0 0 0 0>
      outlineSettings {
       bUseBrushTransparency=false
       roundingType=HalfHeightRadius
       width=0
       color {colorUseRule=UseColor_Specified, specifiedColor=<rgba 0 0 0 0>}
       cornerRadii=<0 0 0 0>
      }
      tintColor {colorUseRule=UseColor_Specified, specifiedColor=<rgba 1 1 1 1>}
      uVRegion {bIsValid=false, max=<0 0>, min=<0 0>}
     }
     selectedBackgroundColor {colorUseRule=UseColor_Foreground, specifiedColor=<rgba 1 0 1 1>}
     shadowColorAndOpacity=<rgba 0 0 0 1>
     shadowOffset=<0 0>
     strikeBrush {
      bIsDynamicallyLoaded=false
      drawAs=Image
      imageType=NoImage
      mirroring=NoMirror
      resourceName=None
      resourceObject=null
      tiling=NoTile
      imageSize=<32 32>
      margin=<ltrb 0 0 0 0>
      outlineSettings {
       bUseBrushTransparency=false
       roundingType=HalfHeightRadius
       width=0
       color {colorUseRule=UseColor_Specified, specifiedColor=<rgba 0 0 0 0>}
       cornerRadii=<0 0 0 0>
      }
      tintColor {colorUseRule=UseColor_Specified, specifiedColor=<rgba 1 1 1 1>}
      uVRegion {bIsValid=false, max=<0 0>, min=<0 0>}
     }
     underlineBrush {
      bIsDynamicallyLoaded=false
      drawAs=Image
      imageType=NoImage
      mirroring=NoMirror
      resourceName=None
      resourceObject=null
      tiling=NoTile
      imageSize=<32 32>
      margin=<ltrb 0 0 0 0>
      outlineSettings {
       bUseBrushTransparency=false
       roundingType=HalfHeightRadius
       width=0
       color {colorUseRule=UseColor_Specified, specifiedColor=<rgba 0 0 0 0>}
       cornerRadii=<0 0 0 0>
      }
      tintColor {colorUseRule=UseColor_Specified, specifiedColor=<rgba 1 1 1 1>}
      uVRegion {bIsValid=false, max=<0 0>, min=<0 0>}
     }
    }
    renderTransform {angle=0, scale=<1 1>, shear=<0 0>, translation=<0 2>}
   }
   slot {class=WidgetSwitcherSlot, properties {verticalAlignment=VAlign_Center, padding=<ltrb 4 0 0 0>}}
  }
  FortVisualAttachment_1 {
   widgetClassPath=&FortVisualAttachment
   parent=VerticalBox
   widget=FortVisualAttachment_1
   properties {renderTransform {angle=0, scale=<1 1>, shear=<0 0>, translation=<18 0>}}
   slot {class=VerticalBoxSlot, properties {horizontalAlignment=HAlign_Left, verticalAlignment=VAlign_Top}}
  }
  SquadMarkerConfirmationWidget {
   widgetClassPath=&/SquadMarkerConfirmationWidget
   bIsVariable=true
   parent=FortVisualAttachment_1
   widget=SquadMarkerConfirmationWidget
   properties {confirmationPlacementInterval=36, renderOpacity=0.96}
   slot {class=SizeBoxSlot, properties {horizontalAlignment=HAlign_Center}}
  }
  OverrideSlot {
   widgetClassPath=&NamedSlot
   bIsVariable=true
   parent=OverrideSwitcher
   widget=OverrideSlot
   properties {bExposeOnInstanceOnly=true}
   slot {class=WidgetSwitcherSlot}
  }
  TutorialHighlight {widgetClassPath=&TutorialHighlight}
 ]
}
